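; Art review helper Blueprint that switches between post-process render pass visualizers such as base color, roughness, normals and level checkers.
; Each mode function calls Reset to disable the other PostProcess components, then enables its own.
; The two LevelChecker variants take an Enabled flag, while the other modes always switch on.
; Reset omits the vertex color component, so that view stays enabled once VertexColor turns it on.
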(fn BaseColor
  (RenderPassVisualizers|Reset)
  (Variables|/Script/Engine.PostProcessComponent|SetbEnabled true (Variables|Default|GetPostProcess_baseColor)))

(fn BaseColorDesaturated
  (RenderPassVisualizers|Reset)
  (Variables|/Script/Engine.PostProcessComponent|SetbEnabled true (Variables|Default|GetPostProcess_BaseColorDesaturated_0)))

(fn LevelChecker
  (params (Enabled :type Boolean :default true))
  (RenderPassVisualizers|Reset)
  (Variables|/Script/Engine.PostProcessComponent|SetbEnabled Enabled (Variables|Default|GetPostProcess_LevelChecker)))

(fn LevelChecker_NoSS
  (params (Enabled :type Boolean :default true))
  (RenderPassVisualizers|Reset)
  (Variables|/Script/Engine.PostProcessComponent|SetbEnabled Enabled (Variables|Default|GetPostProcess_LevelChecker_NoSS)))

(fn MaterialAO
  (RenderPassVisualizers|Reset)
  (Variables|/Script/Engine.PostProcessComponent|SetbEnabled true (Variables|Default|GetPostProcess_MaterialAO)))

(fn Metallic
  (RenderPassVisualizers|Reset)
  (Variables|/Script/Engine.PostProcessComponent|SetbEnabled true (Variables|Default|GetPostProcess_metallic)))

(fn Reset
  (Variables|/Script/Engine.PostProcessComponent|SetbEnabled false (Variables|Default|GetPostProcess_baseColor))
  (Variables|/Script/Engine.PostProcessComponent|SetbEnabled false (Variables|Default|GetPostProcess_BaseColorDesaturated_0))
  (Variables|/Script/Engine.PostProcessComponent|SetbEnabled false (Variables|Default|GetPostProcess_specular))
  (Variables|/Script/Engine.PostProcessComponent|SetbEnabled false (Variables|Default|GetPostProcess_roughness))
  (Variables|/Script/Engine.PostProcessComponent|SetbEnabled false (Variables|Default|GetPostProcess_metallic))
  (Variables|/Script/Engine.PostProcessComponent|SetbEnabled false (Variables|Default|GetPostProcess_worldnormal))
  (Variables|/Script/Engine.PostProcessComponent|SetbEnabled false (Variables|Default|GetPostProcess_subsurfacecolor))
  (Variables|/Script/Engine.PostProcessComponent|SetbEnabled false (Variables|Default|GetPostProcess_SeperateAlpha))
  (Variables|/Script/Engine.PostProcessComponent|SetbEnabled false (Variables|Default|GetPostProcess_MaterialAO))
  (Variables|/Script/Engine.PostProcessComponent|SetbEnabled false (Variables|Default|GetPostProcess_LevelChecker))
  (Variables|/Script/Engine.PostProcessComponent|SetbEnabled false (Variables|Default|GetPostProcess_LevelChecker_NoSS)))

(fn Roughness
  (RenderPassVisualizers|Reset)
  (Variables|/Script/Engine.PostProcessComponent|SetbEnabled true (Variables|Default|GetPostProcess_roughness)))

(fn SeperateAlpha
  (RenderPassVisualizers|Reset)
  (Variables|/Script/Engine.PostProcessComponent|SetbEnabled true (Variables|Default|GetPostProcess_SeperateAlpha)))

(fn Specular
  (RenderPassVisualizers|Reset)
  (Variables|/Script/Engine.PostProcessComponent|SetbEnabled true (Variables|Default|GetPostProcess_specular)))

(fn SubSurfaceColor
  (RenderPassVisualizers|Reset)
  (Variables|/Script/Engine.PostProcessComponent|SetbEnabled true (Variables|Default|GetPostProcess_subsurfacecolor)))

(fn ConstructionScript)

(fn VertexColor
  (RenderPassVisualizers|Reset)
  (Variables|/Script/Engine.PostProcessComponent|SetbEnabled true (Variables|Default|GetPostProcess_vertexcolor)))

(fn WorldNormal
  (RenderPassVisualizers|Reset)
  (Variables|/Script/Engine.PostProcessComponent|SetbEnabled true (Variables|Default|GetPostProcess_worldnormal)))
