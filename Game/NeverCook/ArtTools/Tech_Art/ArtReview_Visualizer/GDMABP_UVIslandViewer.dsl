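; UV island viewer Blueprint for art review that bakes a texture visualizing one skeletal mesh material slot's UV islands.
; RequestTextureForMaterialSlot copies the mesh into a pooled dynamic mesh, deletes other slots' triangles and converts UV islands to polygroups.
; Each island gets a seeded random vertex color, with red overriding islands whose averaged UV centroids match exactly, treated as overlaps.
; Geometry Script bakes the vertex colors into a 2K texture and returns it.
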(event EventBeginPlay)

(event Collision|EventActorBeginOverlap
  (params (OtherActor :type Actor%20Object%20Reference)))

(event EventTick
  (params (DeltaSeconds :type Float%20%28single-precision%29)))

(event Events|EventOnRebuildGeneratedMesh
  (params (TargetMesh :type Dynamic%20Mesh%20Object%20Reference))
  (Variables|Default|SetTarget%20Mesh TargetMesh))

(fn RequestTextureForMaterialSlot
  (params (SkelMesh :type Skeletal%20Mesh%20Object%20Reference) (MaterialSlotIndex :type Integer))
  (bind _centroid (LocalVariables|GetCentroid))
  (bind _polygroupselectioncenter (LocalVariables|GetPolygroupSelectionCenter))
  (bind _returnvalue (Utilities|Array|FindItem _polygroupselectioncenter _centroid))
  (bind _poly_group_ids_out (Variables|Default|GetPoly%20Group%20IDs%20Out))
  (bind _returnvalue_2 (GeometryScript|ListUtils|GetIndexListLength _poly_group_ids_out))
  (bind _polygrouptriangles (LocalVariables|GetPolyGroupTriangles))
  (Development|PrintString "BeginRequestTexture")
  (bind _returnvalue_3 (GeometryScript|Utility|CreateDynamicMeshPool))
  (bind _mesh (GeometryScript|MeshPool|RequestandReleaseComputeMesh _returnvalue_3)
    (:With%20Mesh
      (bind _returnvalue_4 (GeometryScript|SkeletalMesh|CopyMeshfromSkeletalMesh :FromSkeletalMeshAsset SkelMesh :ToDynamicMesh _mesh :RequestedLOD (Utilities|Struct|MakeGeometryScriptMeshReadLOD "MaxAvailable"))
        (:Failure
          (Development|PrintString "FAILURE COPYING MESH!!!"))
        (:Success
          (for _index (range (+ (.ReturnValue (GeometryScript|MeshQueries|GetMaxMaterialID _returnvalue_4)) 1))
            (Development|PrintString (Utilities|String|Append "Attempt Delete: " (Utilities|String|ToString(Integer) _index)))
            (if (not (== MaterialSlotIndex _index))
              (bind _numdeleted (GeometryScript|Materials|DeleteTrianglesbyMaterialID _returnvalue_4 _index))
              (Development|PrintString (Utilities|String|ToString(Integer) _numdeleted))))
          (GeometryScript|Repair|CompactMesh _returnvalue_4)
          (bind _returnvalue_5 (GeometryScript|PolyGroups|ConvertUVIslandsToPolyGroups _returnvalue_4))
          (GeometryScript|PolyGroups|GetPolyGroupIDsInMesh :TargetMesh _returnvalue_5 :PolygroupIDsOut _poly_group_ids_out)
          (Development|PrintString (Utilities|String|ToString(Integer) _returnvalue_2))
          (Utilities|Array|Clear (LocalVariables|GetVertexUVLocationIDList))
          (Utilities|Array|Clear (LocalVariables|GetPolygroupSelectionCenter))
          (for _index_2 (range _returnvalue_2)
            ; Color Polygroup uniquely
            (bind (_selection _returnvalue_6) (GeometryScript|MeshSelection|SelectMeshElementsbyPolygroup :TargetMesh _returnvalue_5 :PolygroupID _index_2 :SelectionType "Polygroups"))
            (GeometryScript|VertexColor|SetMeshSelectionVertexColor :TargetMesh _returnvalue_6 :Selection _selection :Color (Math|Color|MakeColor (Math|Random|RandomFloatfromStream (Math|Random|MakeRandomStream _index_2)) (Math|Random|RandomFloatfromStream (Math|Random|MakeRandomStream _index_2)) (Math|Random|RandomFloatfromStream (Math|Random|MakeRandomStream _index_2))) :bCreateColorSeam false)
            ; Handle Detect Overlaps
            (LocalVariables|SetCentroid "0, 0, 0")
            (LocalVariables|SetPointCounter 0)
            (bind _returnvalue_7 (GeometryScript|PolyGroups|GetTrianglesInPolyGroup :TargetMesh _returnvalue_5 :PolygroupID _index_2 :TriangleIDsOut _polygrouptriangles))
            (for _index_3 (range (GeometryScript|ListUtils|GetIndexListLength _polygrouptriangles))
              ;; Add Point Data to Calculation and track number of points
              (LocalVariables|SetCentroid (Utilities|Operators|Add (LocalVariables|GetCentroid) (GeometryScript|MeshQueries|GetTriangleUVs _returnvalue_7 0 (.ReturnValue (GeometryScript|ListUtils|GetIndexListItem _polygrouptriangles _index_3))) (.UV2 (GeometryScript|MeshQueries|GetTriangleUVs _returnvalue_7 0 (.ReturnValue (GeometryScript|ListUtils|GetIndexListItem _polygrouptriangles _index_3)))) (.UV3 (GeometryScript|MeshQueries|GetTriangleUVs _returnvalue_7 0 (.ReturnValue (GeometryScript|ListUtils|GetIndexListItem _polygrouptriangles _index_3))))))
              (LocalVariables|SetPointCounter (+ (LocalVariables|GetPointCounter) 3)))
            ;; Finalize Centroid Location
            (LocalVariables|SetCentroid (/ (LocalVariables|GetCentroid) (LocalVariables|GetPointCounter)))
            (if (== _returnvalue -1)
              ;; Store as Unique UV location and Polygroup
              (Utilities|Array|Add _polygroupselectioncenter _centroid)
              (Utilities|Array|Add (LocalVariables|GetVertexUVLocationIDList) _index_2)
              (else
                (Development|PrintString "OVERLAPPING UV FOUND!!!!! OVERRIDING COLOR")
                (bind (_selection_2 _returnvalue_8) (GeometryScript|MeshSelection|SelectMeshElementsbyPolygroup :TargetMesh _returnvalue_5 :PolygroupID _index_2 :SelectionType "Polygroups"))
                (GeometryScript|VertexColor|SetMeshSelectionVertexColor :TargetMesh _returnvalue_8 :Selection _selection_2 :Color (1.0 0.0 0.0 1.0) :bCreateColorSeam false) ; Color Current
                ;; Color saved Island
                (bind (_selection_3 _returnvalue_9) (GeometryScript|MeshSelection|SelectMeshElementsbyPolygroup :TargetMesh _returnvalue_5 :PolygroupID (Utilities|Array|Get(acopy) (LocalVariables|GetVertexUVLocationIDList) _returnvalue) :SelectionType "Polygroups"))
                (GeometryScript|VertexColor|SetMeshSelectionVertexColor :TargetMesh _returnvalue_9 :Selection _selection_3 :Color (1.0 0.0 0.0 1.0) :bCreateColorSeam false) ; Color Current
              )))
          (bind _returnvalue_10 (GeometryScript|Bake|BakeTexture :TargetMesh _returnvalue_5 :TargetOptions (struct :TargetUVLayer 0) :SourceMesh _returnvalue_5 :BakeTypes (Utilities|Array|MakeArray (GeometryScript|Bake|MakeBakeTypeVertexColor)) :BakeOptions (struct :Resolution "Resolution2048" :BitDepth "ChannelBits8" :GutterSize 4 :SamplesPerPixel "Sample1" :SampleFilterMask "" :FilteringType "BSpline" :ProjectionDistance 0.0 :bProjectionInWorldSpace false)))
          (Development|PrintString (Utilities|GetDisplayName (Utilities|Array|Get(acopy) _returnvalue_10)))
          (return (Utilities|Array|Get(acopy) _returnvalue_10)))))))

(fn ConstructionScript)
