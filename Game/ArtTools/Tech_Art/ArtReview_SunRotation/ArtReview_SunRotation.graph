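# Editor utility widget for art review that rotates the time-of-day sun actor with a sun-icon slider.
# GetTODM casts the global time-of-day manager to FortDaySequenceActor; RotateTOD sets its yaw to slider value plus InitialRotation, keeping pitch and roll.
# Construct caches that starting yaw; Resetslider recaptures it and zeroes the slider.
# Slate throttling is disabled during slider mouse capture and re-enabled on release.
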EditorUtilityWidgetBlueprint /Game/ArtTools/Tech_Art/ArtReview_SunRotation/ArtReview_SunRotation
graphs [
 EventGraph {nodes [
   n1 = "On Value Changed (Slider_20)"!() -> (OutputDelegate, then, Value:float) {is=title K2Node_ComponentBoundEvent}
   > PrintString(InString="To String (Float)"(InDouble=@n1.Value))
   > n2 = Get TODM!(In:Exec) -> (Out:Exec, TODM:Actor Object Reference) {is=title K2Node_MacroInstance}
   > RotateTOD!(execute, self, Array:Actor Object Reference=@n2, NewRotation:double=@n1.Value) -> (then)
   "On Mouse Capture Begin (Slider_20)"() {is=title K2Node_ComponentBoundEvent}
   > ExecuteConsoleCommand(Command=Slate.bAllowThrottling 0)
   "On Mouse Capture End (Slider_20)"() {is=title K2Node_ComponentBoundEvent}
   > ExecuteConsoleCommand(Command=Slate.bAllowThrottling 1)
   Resetslider() {is=title K2Node_CustomEvent}
   > n3 = Get TODM()
   > Set InitialRotation(InitialRotation:float=BreakRotator(InRot=Get Actor Rotation!(self:Actor Object Reference=@n3) -> (ReturnValue:Rotator) {is=title K2Node_CallFunction}).Yaw) -> (Output_Get:float)
   > SetValue!(execute, self:Slider Object Reference=Get Slider_20() -> (Slider_20:Slider Object Reference), InValue:float=0) -> (then)
   Event Construct()
   > PrintString(InString="~~~~~~~~~~~~~~~~~~~~~~~~~~~~~~~~~~")
   > n4 = Get TODM()
   n5 = BreakRotator(InRot=Get Actor Rotation(self=@n4))
   > Set InitialRotation(InitialRotation=@n5.Yaw)
   > PrintString(InString="To String (Float)"(InDouble=@n5.Yaw))
   SetValue(self=Get Slider_20(), InValue=0)
   Event PreConstruct()
   Event Tick()
  ]}
 GetTODM {nodes [
   Inputs!() -> (In:Exec)
   > n1 = Cast To FortDaySequenceActor!(
    execute
    Object:Fort Time Of Day Manager Interface Interface=GetGlobalTimeOfDayManager!(WorldContextObject:Object Reference) -> (ReturnValue:Fort Time Of Day Manager Interface Interface)
   ) -> (then, CastFailed, AsFort Day Sequence Actor:Fort Day Sequence Actor Object Reference) {is=title K2Node_DynamicCast}
   > Outputs!(Out:Exec, TODM:Actor Object Reference=@n1)
  ]}
 RotateTOD {nodes [
   n1 = RotateTOD!() -> (then, Array:Actor Object Reference, NewRotation:double=~) {is=title K2Node_FunctionEntry}
   n2 = BreakRotator(InRot=Get Actor Rotation(self=@n1))
   n3 = Get InitialRotation() -> (InitialRotation:float)
   > Set Actor Rotation!(
    execute
    self:Actor Object Reference=@n1
    NewRotation:Rotator=MakeRotator(Roll=@n2, Pitch=@n2.Pitch, Yaw=float + float(A=@n1.NewRotation, B=@n3))
    bTeleportPhysics:bool=false
   ) -> (then, ReturnValue:bool) {is=title K2Node_CallFunction}
   > Return Node()
  ]}
]
variables [
 InitialRotation {type=number}
 Out Actors {type="array<BP_Asteria_DaySequence_C>", default []}
 TODM_Rotation {type=Rotator}
]
widgets {
 info {rootWidgetClass=&Overlay, widgetCount=2}
 widgets [
  Overlay_14 {widgetClassPath=&Overlay, widget=Overlay_14}
  Slider_20 {
   widgetClassPath=/Script/UMG.Slider
   bIsVariable=true
   parent=Overlay_14
   widget=Slider_20
   properties {
    maxValue=360
    sliderHandleColor=<rgba 1 1 0 1>
    widgetStyle {
     barThickness=4
     disabledBarImage {
      bIsDynamicallyLoaded=false
      drawAs=RoundedBox
      imageType=NoImage
      mirroring=NoMirror
      resourceName=None
      resourceObject=null
      tiling=NoTile
      imageSize=<32 32>
      margin=<ltrb 0 0 0 0>
      outlineSettings {
       bUseBrushTransparency=false
       roundingType=FixedRadius
       width=1
       color {colorUseRule=UseColor_ColorTable, specifiedColor=<rgba 0 0 0 0>}
       cornerRadii=<2 2 2 2>
      }
      tintColor {colorUseRule=UseColor_Specified, specifiedColor=<rgba 1 1 1 1>}
      uVRegion {bIsValid=false, max=<0 0>, min=<0 0>}
     }
     disabledThumbImage {
      bIsDynamicallyLoaded=false
      drawAs=RoundedBox
      imageType=NoImage
      mirroring=NoMirror
      resourceName=None
      resourceObject=null
      tiling=NoTile
      imageSize=<32 32>
      margin=<ltrb 0 0 0 0>
      outlineSettings {
       bUseBrushTransparency=false
       roundingType=HalfHeightRadius
       width=0
       color {colorUseRule=UseColor_Specified, specifiedColor=<rgba 0 0 0 0>}
       cornerRadii=<0 0 0 0>
      }
      tintColor {colorUseRule=UseColor_Specified, specifiedColor=<rgba 0.5271 0.5271 0.5271 1>}
      uVRegion {bIsValid=false, max=<0 0>, min=<0 0>}
     }
     hoveredBarImage {
      bIsDynamicallyLoaded=false
      drawAs=RoundedBox
      imageType=NoImage
      mirroring=NoMirror
      resourceName=None
      resourceObject=null
      tiling=NoTile
      imageSize=<0 0>
      margin=<ltrb 0 0 0 0>
      outlineSettings {
       bUseBrushTransparency=false
       roundingType=FixedRadius
       width=1
       color {colorUseRule=UseColor_ColorTable, specifiedColor=<rgba 0 0 0 0>}
       cornerRadii=<2 2 2 2>
      }
      tintColor {colorUseRule=UseColor_Specified, specifiedColor=<rgba 0.0048 0.0048 0.0048 1>}
      uVRegion {bIsValid=false, max=<0 0>, min=<0 0>}
     }
     hoveredThumbImage {
      bIsDynamicallyLoaded=false
      drawAs=RoundedBox
      imageType=NoImage
      mirroring=NoMirror
      resourceName=None
      resourceObject=/BRCosmetics/2dAssets/Emoji/Season8/T-Emote-Icons-Season8-sun
      tiling=NoTile
      imageSize=<28 28>
      margin=<ltrb 0 0 0 0>
      outlineSettings {
       bUseBrushTransparency=false
       roundingType=HalfHeightRadius
       width=0
       color {colorUseRule=UseColor_Specified, specifiedColor=<rgba 0 0 0 0>}
       cornerRadii=<0 0 0 0>
      }
      tintColor {colorUseRule=UseColor_Specified, specifiedColor=<rgba 1 1 1 1>}
      uVRegion {bIsValid=false, max=<0 0>, min=<0 0>}
     }
     normalBarImage {
      bIsDynamicallyLoaded=false
      drawAs=RoundedBox
      imageType=NoImage
      mirroring=NoMirror
      resourceName=None
      resourceObject=null
      tiling=NoTile
      imageSize=<32 32>
      margin=<ltrb 0 0 0 0>
      outlineSettings {
       bUseBrushTransparency=false
       roundingType=FixedRadius
       width=1
       color {colorUseRule=UseColor_ColorTable, specifiedColor=<rgba 0 0 0 0>}
       cornerRadii=<2 2 2 2>
      }
      tintColor {colorUseRule=UseColor_Specified, specifiedColor=<rgba 0.0048 0.0048 0.0048 1>}
      uVRegion {bIsValid=false, max=<0 0>, min=<0 0>}
     }
     normalThumbImage {
      bIsDynamicallyLoaded=false
      drawAs=RoundedBox
      imageType=NoImage
      mirroring=NoMirror
      resourceName=None
      resourceObject=&T-Emote-Icons-Season8-sun
      tiling=NoTile
      imageSize=<20 20>
      margin=<ltrb 0 0 0 0>
      outlineSettings {
       bUseBrushTransparency=false
       roundingType=HalfHeightRadius
       width=0
       color {colorUseRule=UseColor_Specified, specifiedColor=<rgba 0 0 0 0>}
       cornerRadii=<0 0 0 0>
      }
      tintColor {colorUseRule=UseColor_Specified, specifiedColor=<rgba 0.5271 0.5271 0.5271 1>}
      uVRegion {bIsValid=false, max=<0 0>, min=<0 0>}
     }
    }
   }
   slot {class=OverlaySlot, properties {horizontalAlignment=HAlign_Fill}}
  }
 ]
}
