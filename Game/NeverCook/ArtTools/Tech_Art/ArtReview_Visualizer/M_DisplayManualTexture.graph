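# Unlit material in the NeverCook art review visualizer tools, named for displaying a manually chosen texture.
# Its only node is a 2D texture sample parameter named Texture, which material instances can override; no link to a material output is shown.
# Usage flags cover skeletal meshes, clothing and mesh deformers, with automatic usage setting in the editor turned off.
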Material /Game/NeverCook/ArtTools/Tech_Art/ArtReview_Visualizer/M_DisplayManualTexture
graph {
 properties {
  bAutomaticallySetUsageInEditor=false
  bUsedWithClothing=true
  bUsedWithMeshDeformer=true
  bUsedWithSkeletalMesh=true
  shadingModel=MSM_Unlit
  usedShadingModels=Unlit
 }
 nodes [TextureSampleParameter2D() {parameterName=Texture}]
}
