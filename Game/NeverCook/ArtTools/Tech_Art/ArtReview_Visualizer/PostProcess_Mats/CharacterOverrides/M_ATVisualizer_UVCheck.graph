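# Unlit UV-check override material for characters in the ArtReview Visualizer art tool, built around a checker pattern to expose texture mapping.
# A single texture sample reads the Checker texture from the Lowpoly_Turntable content at default UVs, with no other graph logic.
# Usage is set explicitly for skeletal meshes, clothing and mesh deformers, with automatic usage setting in the editor turned off.
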Material /Game/NeverCook/ArtTools/Tech_Art/ArtReview_Visualizer/PostProcess_Mats/CharacterOverrides/M_ATVisualizer_UVCheck
graph {
 properties {
  bAutomaticallySetUsageInEditor=false
  bUsedWithClothing=true
  bUsedWithMeshDeformer=true
  bUsedWithSkeletalMesh=true
  shadingModel=MSM_Unlit
  usedShadingModels=Unlit
 }
 nodes [TextureSample() {texture=/Lowpoly_Turntable/Lowpoly_TT_Assets/BlueprintAssets/Checker}]
}
