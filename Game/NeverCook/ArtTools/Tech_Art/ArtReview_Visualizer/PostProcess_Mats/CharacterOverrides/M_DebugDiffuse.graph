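# Unlit debug material for the Art Review Visualizer's character overrides, kept under the NeverCook art tools folder.
# Its only node is a Diffuse texture parameter with masks sampling and a black mask default, with no link to a material output shown.
# Usage flags enable it on skeletal meshes and mesh deformers.
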Material /Game/NeverCook/ArtTools/Tech_Art/ArtReview_Visualizer/PostProcess_Mats/CharacterOverrides/M_DebugDiffuse
graph {
 properties {bUsedWithMeshDeformer=true, bUsedWithSkeletalMesh=true, pixelDepthOffsetMode=PDOM_Legacy, shadingModel=MSM_Unlit, usedShadingModels=Unlit}
 nodes [TextureSampleParameter2D() {parameterName=Diffuse, samplerType=SAMPLERTYPE_Masks, texture=&T_BlackMask}]
}
