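# M_DebugSRM_G is an unlit debug material for skeletal meshes that isolates the green channel of character SRM mask textures for the art review visualizer.
# It adds the RGB of the SpecularMasks and SRM texture parameters, then masks the sum to green.
# Both parameters default to a black mask, so an unset one adds nothing, which likely lets one material cover either parameter name.
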Material /Game/NeverCook/ArtTools/Tech_Art/ArtReview_Visualizer/PostProcess_Mats/CharacterOverrides/M_DebugSRM_G
graph {
 properties {bUsedWithMeshDeformer=true, bUsedWithSkeletalMesh=true, pixelDepthOffsetMode=PDOM_Legacy, shadingModel=MSM_Unlit, usedShadingModels=Unlit}
 nodes [ComponentMask(None=Add(
    A=TextureSampleParameter2D() {parameterName=SpecularMasks, samplerType=SAMPLERTYPE_Masks, texture=&T_BlackMask}.RGB
    B=TextureSampleParameter2D() {parameterName=SRM, samplerType=SAMPLERTYPE_Masks, texture=&T_BlackMask}.RGB
   )) {g=true}]
}
