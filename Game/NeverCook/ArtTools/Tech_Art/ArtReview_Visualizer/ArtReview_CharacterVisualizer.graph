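# Editor utility widget for art review that shows debug views of a Fortnite player pawn's textures and UVs.
# Buttons apply local material overrides copying chosen texture parameters, or swap in per-slot UV island textures, saving original materials for Restore.
# Checkboxes and sliders drive a Texture Graph replacing specular, metallic or roughness channels in head, body and backpack SRM/SpecularMasks textures.
# The target is the selected pawn, falling back to one in the PIE world.
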EditorUtilityWidgetBlueprint /Game/NeverCook/ArtTools/Tech_Art/ArtReview_Visualizer/ArtReview_CharacterVisualizer
graphs [
 ApplyCustomChannelOverride {nodes [
   n1 = ApplyCustomChannelOverride!() -> (then, PartType:EFortCustomPartType Enum=Head, New Textures:Array of Texture Object References) {is=title K2Node_FunctionEntry}
   > n2 = GetPawn!(execute, self) -> (then, Output_Get:Fort Player Pawn Object Reference)
   n3 = GetSkeletalMeshForPartType(self=@n2, PartType=@n1)
   > n4 = For Each Loop(Array:Array of Names=GetMaterialSlotNames!(self:Primitive Component Object Reference=@n3) -> (ReturnValue:Array of Names)) -> (Array Element:Name)
   > n5 = Get Dynamic Material Instance(SkeletalMeshComponent=@n3, ElementIndex=@n4.Array Index)
   n6 = "Get (a copy)"(Array:Array of Texture Object References=@n1.New Textures, Dimension 1=@n4.Array Index) -> (Output:Texture Object Reference)
   > SetTextureParameterValue(self=@n5, ParameterName=SRM, Value=@n6)
   > SetTextureParameterValue(self=@n5, ParameterName=SpecularMasks, Value=@n6)
  ]}
 ConfigureTextureTransfer {nodes [
   n1 = ConfigureTextureTransfer!() -> (then, MaterialTextureOverride:Material Interface Object Reference, TextureSlotDisplayName:Name) {is=title K2Node_FunctionEntry}
   > SetTextureOverride_ApplyOverride!(
    execute
    self
    Input:Material Interface Object Reference=@n1
    MaterialParamsToCopy:"Fort Pawn Material Override Copied Parameters Structure (by ref)"=Make Fort Pawn Material Override Copied Parameters!(
     ScalarParamNames:Array of Names
     VectorParamNames:Array of Names
     TextureParamNames:Array of Names=Make Array("[0]":Name=@n1.TextureSlotDisplayName) -> (Array:Array of Names)
    ) -> (FortPawnMaterialOverrideCopiedParameters:Fort Pawn Material Override Copied Parameters Structure) {is=title K2Node_MakeStruct}
   ) -> (then)
  ]}
 CreateRenderTargetMap {nodes [
   CreateRenderTargetMap() {is=title K2Node_FunctionEntry}
   > n1 = Sequence!(execute) -> (then_0, then_1, then_2:Exec)
   > n2 = For Each Loop(Array:Array of Texture Object References=Get HeadTextures() -> (HeadTextures:Array of Texture Object References)) -> (Array Element:Texture Object Reference)
   > n3 = GenerateRenderTarget!(execute, self, BaseTexture:Texture Object Reference=@n2) -> (then, NewParam:Texture Object Reference)
   > Set Array Elem!(
    execute
    TargetArray:Array of Texture Object References=Get ModHeadTextures() -> (ModHeadTextures:Array of Texture Object References)
    Index:int=@n2.Array Index
    Item:"Texture Object Reference (by ref)"=@n3
    bSizeToFit:bool=true
   ) -> (then)
   n4 = For Each Loop(Exec=@n1.then_1, Array=Get BodyTextures() -> (BodyTextures:Array of Texture Object References))
   > n5 = GenerateRenderTarget(BaseTexture=@n4)
   > Set Array Elem(TargetArray=Get ModBodyTextures() -> (ModBodyTextures:Array of Texture Object References), Index=@n4.Array Index, Item=@n5, bSizeToFit=true)
   n6 = For Each Loop(Exec=@n1.then_2, Array=Get BackPackTextures() -> (BackPackTextures:Array of Texture Object References))
   > n7 = GenerateRenderTarget(BaseTexture=@n6)
   > Set Array Elem(
    TargetArray=Get ModBackpackTextures() -> (ModBackpackTextures:Array of Texture Object References)
    Index=@n6.Array Index
    Item=@n7
    bSizeToFit=true
   )
  ]}
 EventGraph {nodes [
   "On Clicked (Button_SRM_B)"!() -> (OutputDelegate, then, widget:User Widget Object Reference, event:Pointer Event Structure) {is=title K2Node_ComponentBoundEvent}
   > SetTextureOverride_ApplyOverride(
    Input=/Game/NeverCook/ArtTools/Tech_Art/ArtReview_Visualizer/PostProcess_Mats/CharacterOverrides/M_DebugSRM_B
    MaterialParamsToCopy=Get SRMSettings() -> (SRMSettings:Fort Pawn Material Override Copied Parameters Structure)
   )
   "On Clicked (Button_SRM_G)"!() -> (OutputDelegate, then, widget:User Widget Object Reference, event:Pointer Event Structure) {is=title K2Node_ComponentBoundEvent}
   > SetTextureOverride_ApplyOverride(
    Input=/Game/NeverCook/ArtTools/Tech_Art/ArtReview_Visualizer/PostProcess_Mats/CharacterOverrides/M_DebugSRM_G
    MaterialParamsToCopy=Get SRMSettings()
   )
   n1 = "On Check State Changed (AdvancedMaps)"!() -> (OutputDelegate, then, bIsChecked:bool) {is=title K2Node_ComponentBoundEvent}
   > n2 = Branch(Condition=@n1.bIsChecked)
   n3 = Get AdvancedSliders() -> (AdvancedSliders:Horizontal Box Object Reference)
   > SetVisibility(self=@n3, InVisibility=Visible)
   > n4 = GetPawn()
   > n5 = GetTexturesForPartType!(execute, self, Target:Fort Player Pawn Object Reference=@n4, PartType:EFortCustomPartType Enum=Head) -> (
    then
    TexturesViaSlot1:Array of Texture Object References
    PartSkelMesh:Skeletal Mesh Component Object Reference
    MaterialsForPartType:Array of Material Interface Object References
   )
   > Set HeadTextures(HeadTextures:Array of Texture Object References=@n5) -> (Output_Get:Array of Texture Object References)
   > n6 = GetTexturesForPartType(Target=@n4)
   > Set BodyTextures(BodyTextures:Array of Texture Object References=@n6) -> (Output_Get:Array of Texture Object References)
   > n7 = GetTexturesForPartType(Target=@n4, PartType=Backpack)
   > Set BackPackTextures(BackPackTextures:Array of Texture Object References=@n7) -> (Output_Get:Array of Texture Object References)
   n8 = Sequence!(execute=@n2.else) -> (then_0, then_1)
   > SetVisibility(self=@n3, InVisibility=Collapsed)
   > SetCheckedState!(
    execute
    self:Check Box Object Reference=[
     Get Specular() -> (Specular:Editor Utility Check Box Object Reference)
     Get Roughness() -> (Roughness:Editor Utility Check Box Object Reference)
     Get Metallic() -> (Metallic:Editor Utility Check Box Object Reference)
    ]
    InCheckedState:ECheckBoxState Enum=Unchecked
   ) -> (then)
   n9 = Get SpecSlider() -> (SpecSlider:Editor Utility Slider Object Reference)
   n10 = Get MetalSlider() -> (MetalSlider:Editor Utility Slider Object Reference)
   n11 = Get RoughSlider() -> (RoughSlider:Editor Utility Slider Object Reference)
   > SetIsEnabled(self=[@n9, @n10, @n11], bInIsEnabled=false)
   > Set Value!(execute, Value:float=0.5, self:Slider Object Reference=@n11) -> (then, Output_Get:float)
   > Set Value(Value=0.5, self=@n10)
   > Set Value(Value=0.5, self=@n9)
   ApplyCustomChannelOverride!(execute=@n8.then_1, self, PartType:EFortCustomPartType Enum=Head, New Textures:Array of Texture Object References=Get HeadTextures()) -> (then) {is=title K2Node_CallFunction}
   > ApplyCustomChannelOverride(New Textures=Get BodyTextures())
   > ApplyCustomChannelOverride(PartType=Backpack, New Textures=Get BackPackTextures())
   n12 = "On Check State Changed (Specular)"!() -> (OutputDelegate, then, bIsChecked:bool) {is=title K2Node_ComponentBoundEvent}
   > Set OverrideSpecular?(OverrideSpecular?:bool=@n12.bIsChecked) -> (Output_Get:bool)
   > n13 = SetIsEnabled(self=Get SpecSlider(), bInIsEnabled=@n12.bIsChecked)
   GenerateRenderTargetTexture(execute=[@n13.then, @n17.then, @n15.then])
   "On Clicked (Button_MMask_R)"!() -> (OutputDelegate, then, widget:User Widget Object Reference, event:Pointer Event Structure) {is=title K2Node_ComponentBoundEvent}
   > SetTextureOverride_ApplyOverride(
    Input=/Game/NeverCook/ArtTools/Tech_Art/ArtReview_Visualizer/PostProcess_Mats/CharacterOverrides/M_DebugMMask_R
    MaterialParamsToCopy=Get MMaskSettings() -> (MMaskSettings:Fort Pawn Material Override Copied Parameters Structure)
   )
   "On Clicked (Button_MMask_G)"!() -> (OutputDelegate, then, widget:User Widget Object Reference, event:Pointer Event Structure) {is=title K2Node_ComponentBoundEvent}
   > SetTextureOverride_ApplyOverride(
    Input=/Game/NeverCook/ArtTools/Tech_Art/ArtReview_Visualizer/PostProcess_Mats/CharacterOverrides/M_DebugMMask_G
    MaterialParamsToCopy=Get MMaskSettings()
   )
   "On Clicked (Button_MMask_B)"!() -> (OutputDelegate, then, widget:User Widget Object Reference, event:Pointer Event Structure) {is=title K2Node_ComponentBoundEvent}
   > SetTextureOverride_ApplyOverride(
    Input=/Game/NeverCook/ArtTools/Tech_Art/ArtReview_Visualizer/PostProcess_Mats/CharacterOverrides/M_DebugMMask_B
    MaterialParamsToCopy=Get MMaskSettings()
   )
   n14 = "On Check State Changed (Roughness)"!() -> (OutputDelegate, then, bIsChecked:bool) {is=title K2Node_ComponentBoundEvent}
   > Set OverrideRoughness?(OverrideRoughness?:bool=@n14.bIsChecked) -> (Output_Get:bool)
   > n15 = SetIsEnabled(self=Get RoughSlider(), bInIsEnabled=@n14.bIsChecked)
   n16 = "On Check State Changed (Metallic)"!() -> (OutputDelegate, then, bIsChecked:bool) {is=title K2Node_ComponentBoundEvent}
   > Set OverrideMetallic?(OverrideMetallic?:bool=@n16.bIsChecked) -> (Output_Get:bool)
   > n17 = SetIsEnabled(self=Get MetalSlider(), bInIsEnabled=@n16.bIsChecked)
   "On Mouse Capture End (SpecSlider)"() {is=title K2Node_ComponentBoundEvent}
   > n18 = Set SpecSliderValue(SpecSliderValue:float=GetValue!(self:Slider Object Reference=Get SpecSlider()) -> (ReturnValue:float)) -> (Output_Get:float)
   > n19 = "SetText (Text)"(
    self=Get SpecValue() -> (SpecValue:Text Object Reference)
    InText="To Text (Float)"(Value=@n18, MinimumIntegralDigits=1, MaximumIntegralDigits=324, MinimumFractionalDigits=0, MaximumFractionalDigits=3)
   )
   GenerateRenderTargetTexture(execute=[@n23.then, @n19.then, @n21.then])
   "On Mouse Capture End (MetalSlider)"() {is=title K2Node_ComponentBoundEvent}
   > n20 = Set MetallicSliderValue(MetallicSliderValue:float=GetValue(self=Get MetalSlider())) -> (Output_Get:float)
   > n21 = "SetText (Text)"(
    self=Get MetalValue() -> (MetalValue:Text Object Reference)
    InText="To Text (Float)"(Value=@n20, MinimumIntegralDigits=1, MaximumIntegralDigits=324, MinimumFractionalDigits=0, MaximumFractionalDigits=3)
   )
   "On Mouse Capture End (RoughSlider)"() {is=title K2Node_ComponentBoundEvent}
   > n22 = Set RoughnessSliderValue(RoughnessSliderValue:float=GetValue(self=Get RoughSlider())) -> (Output_Get:float)
   > n23 = "SetText (Text)"(
    self=Get RoughValue() -> (RoughValue:Text Object Reference)
    InText="To Text (Float)"(Value=@n22, MinimumIntegralDigits=1, MaximumIntegralDigits=324, MinimumFractionalDigits=0, MaximumFractionalDigits=3)
   )
   "On Clicked (Button_MMask)"!() -> (OutputDelegate, then, widget:User Widget Object Reference, event:Pointer Event Structure) {is=title K2Node_ComponentBoundEvent}
   > SetTextureOverride_ApplyOverride(
    Input=/Game/NeverCook/ArtTools/Tech_Art/ArtReview_Visualizer/PostProcess_Mats/CharacterOverrides/M_DebugMMask
    MaterialParamsToCopy=Get MMaskSettings()
   )
   "On Clicked (Button_UVCheck)"!() -> (OutputDelegate, then, widget:User Widget Object Reference, event:Pointer Event Structure) {is=title K2Node_ComponentBoundEvent}
   > SetTextureOverride_ApplyOverride(Input=/Game/NeverCook/ArtTools/Tech_Art/ArtReview_Visualizer/PostProcess_Mats/CharacterOverrides/M_ATVisualizer_UVCheck)
   "On Clicked (Button_UVIslands)"!() -> (OutputDelegate, then, widget:User Widget Object Reference, event:Pointer Event Structure) {is=title K2Node_ComponentBoundEvent}
   > n24 = Branch(Condition=Get IsDoingManualMaterialSwap?() -> (IsDoingManualMaterialSwap?:bool))
   Set IsDoingManualMaterialSwap?(execute=@n24.else, IsDoingManualMaterialSwap?:bool=true) -> (Output_Get:bool)
   > n25 = GetPIEWorlds!(execute, bIncludeDedicatedServer:bool=false) -> (then, ReturnValue:Array of World Object References)
   n26 = "Get (a copy)"(Array:Array of World Object References=@n25) -> (Output:World Object Reference)
   > Is Valid(InputObject=@n26)
   > n27 = GetPawn()
   > n28 = Sequence!(execute) -> (then_0, then_1, then_2:Exec)
   > n29 = GetTexturesForPartType(Target=@n27)
   > Set OGBodyMaterials(OGBodyMaterials:Array of Material Interface Object References=@n29.MaterialsForPartType) -> (Output_Get:Array of Material Interface Object References)
   > GenerateAndApplyCustomUVIslandTextures!(
    execute
    self
    SkelMeshRef:Skeletal Mesh Component Object Reference=@n29.PartSkelMesh
    MaterialsForPartType:Array of Material Interface Object References=@n29.MaterialsForPartType
    WorldContextObject:Object Reference=@n26
   ) -> (then)
   n30 = GetTexturesForPartType(execute=@n28.then_1, Target=@n27, PartType=Head)
   > Set OGHeadMaterials(OGHeadMaterials:Array of Material Interface Object References=@n30.MaterialsForPartType) -> (Output_Get:Array of Material Interface Object References)
   > GenerateAndApplyCustomUVIslandTextures(SkelMeshRef=@n30.PartSkelMesh, MaterialsForPartType=@n30.MaterialsForPartType, WorldContextObject=@n26)
   n31 = GetTexturesForPartType(execute=@n28.then_2, Target=@n27, PartType=Backpack)
   > Set OGBackpackMaterials(OGBackpackMaterials:Array of Material Interface Object References=@n31.MaterialsForPartType) -> (Output_Get:Array of Material Interface Object References)
   > GenerateAndApplyCustomUVIslandTextures(SkelMeshRef=@n31.PartSkelMesh, MaterialsForPartType=@n31.MaterialsForPartType, WorldContextObject=@n26)
   "On Clicked (Button_SRM_R)"!() -> (OutputDelegate, then, widget:User Widget Object Reference, event:Pointer Event Structure) {is=title K2Node_ComponentBoundEvent}
   > SetTextureOverride_ApplyOverride(
    Input=/Game/NeverCook/ArtTools/Tech_Art/ArtReview_Visualizer/PostProcess_Mats/CharacterOverrides/M_DebugSRM_R
    MaterialParamsToCopy=Get SRMSettings()
   )
   "On Clicked (Button_FXMask)"!() -> (OutputDelegate, then, widget:User Widget Object Reference, event:Pointer Event Structure) {is=title K2Node_ComponentBoundEvent}
   > SetTextureOverride_ApplyOverride(
    Input=/Game/NeverCook/ArtTools/Tech_Art/ArtReview_Visualizer/PostProcess_Mats/CharacterOverrides/M_DebugSkinFX
    MaterialParamsToCopy=Get FXMaskSettings() -> (FXMaskSettings:Fort Pawn Material Override Copied Parameters Structure)
   )
   "On Clicked (Button_Diffuse)"!() -> (OutputDelegate, then, widget:User Widget Object Reference, event:Pointer Event Structure) {is=title K2Node_ComponentBoundEvent}
   > ConfigureTextureTransfer!(
    execute
    self
    MaterialTextureOverride:Material Interface Object Reference=/Game/NeverCook/ArtTools/Tech_Art/ArtReview_Visualizer/PostProcess_Mats/CharacterOverrides/M_DebugDiffuse
    TextureSlotDisplayName:Name=Diffuse
   ) -> (then) {is=title K2Node_CallFunction}
   "On Clicked (Button_SRM)"!() -> (OutputDelegate, then, widget:User Widget Object Reference, event:Pointer Event Structure) {is=title K2Node_ComponentBoundEvent}
   > SetTextureOverride_ApplyOverride(
    Input=/Game/NeverCook/ArtTools/Tech_Art/ArtReview_Visualizer/PostProcess_Mats/CharacterOverrides/M_DebugSRM
    MaterialParamsToCopy=Get SRMSettings()
   )
   "On Clicked (Button_Fuzz)"!() -> (OutputDelegate, then, widget:User Widget Object Reference, event:Pointer Event Structure) {is=title K2Node_ComponentBoundEvent}
   > ConfigureTextureTransfer(
    MaterialTextureOverride=/Game/NeverCook/ArtTools/Tech_Art/ArtReview_Visualizer/PostProcess_Mats/CharacterOverrides/M_DebugClothFuzz
    TextureSlotDisplayName=ClothFuzz Texture
   )
   "On Clicked (Button_Normals)"!() -> (OutputDelegate, then, widget:User Widget Object Reference, event:Pointer Event Structure) {is=title K2Node_ComponentBoundEvent}
   > GetPawn()
   n32 = Get As Fort Player Pawn() -> (As Fort Player Pawn:Fort Player Pawn Object Reference)
   > RemoveMaterialOverrideLocal!(
    execute
    self:Fort Player Pawn Object Reference=@n32
    MaterialOverrideId:"Guid Structure (by ref)"=Get MaterialOverride() -> (MaterialOverride:Guid Structure)
   ) -> (then, ReturnValue:bool)
   > n33 = Apply Material Override Local with Flags!(
    execute
    self:Fort Player Pawn Object Reference=@n32
    Material:"Material Interface Soft Object Reference (by ref)"=Make Soft Reference!(Input:Material Interface Object Reference=Get HoloMaterialInst() -> (HoloMaterialInst:Material Interface Object Reference)) -> (Output:Material Interface Soft Object Reference)
    MaterialParamsToCopy:"Fort Pawn Material Override Copied Parameters Structure (by ref)"=Get Material Params to Copy() -> (Material Params to Copy:Fort Pawn Material Override Copied Parameters Structure)
    Priority:float=1
    Flags:Bitmask Integer=15
   ) -> (then, ReturnValue:Guid Structure)
   > Set MaterialOverride(MaterialOverride:Guid Structure=@n33) -> (Output_Get:Guid Structure)
   "On Clicked (Button_Restore)"!() -> (OutputDelegate, then, widget:User Widget Object Reference, event:Pointer Event Structure) {is=title K2Node_ComponentBoundEvent}
   > n34 = GetPawn()
   > n35 = Branch(Condition=Get IsDoingManualMaterialSwap?())
   > RestoreToOGMaterials()
   > n36 = Set IsDoingManualMaterialSwap?(IsDoingManualMaterialSwap?=false)
   RemoveMaterialOverrideLocal(execute=[@n35.else, @n36.then], self=@n34, MaterialOverrideId=Get MaterialOverride())
   GenerateRenderTargetTexture() {is=title K2Node_CustomEvent}
   > Clear(TargetArray:Array of Texture Object References=Get ModHeadTextures())
   > Clear(TargetArray=Get ModBodyTextures())
   > Clear(TargetArray=Get ModBackpackTextures())
   > CreateRenderTargetMap() {is=title K2Node_CallFunction}
   > ApplyTexture()
   ApplyTexture() {is=title K2Node_CustomEvent}
   > ApplyCustomChannelOverride(PartType=Head, New Textures=Get ModHeadTextures())
   > ApplyCustomChannelOverride(New Textures=Get ModBodyTextures())
   > ApplyCustomChannelOverride(PartType=Backpack, New Textures=Get ModBackpackTextures())
   ProcessHead() {is=title K2Node_CustomEvent}
   n37 = Length(TargetArray:Array of Texture Object References=Get HeadTextures())
   > Set TextureProcessingSize(TextureProcessingSize:int=@n37) -> (Output_Get:int=0)
   > Resize!(execute, TargetArray:Array of Texture Object References=Get ModHeadTextures(), Size:int=@n37) -> (then) {is=title K2Node_CallArrayFunction}
   > Set AsyncCounter(AsyncCounter:int=0) -> (Output_Get:int=0)
   > ProcessHeadAsync()
   ProcessBody() {is=title K2Node_CustomEvent}
   n38 = Length(TargetArray=Get BodyTextures())
   > Set TextureProcessingSize(TextureProcessingSize=@n38) -> (Output_Get=0)
   > Resize(TargetArray=Get ModBodyTextures(), Size=@n38)
   > Set AsyncCounter(AsyncCounter=0) -> (Output_Get=0)
   > ProcessBodyAsync()
   ProcessBackpack() {is=title K2Node_CustomEvent}
   n39 = Length(TargetArray=Get BackPackTextures())
   > Set TextureProcessingSize(TextureProcessingSize=@n39) -> (Output_Get=0)
   > Resize(TargetArray=Get ModBackpackTextures(), Size=@n39)
   > Set AsyncCounter(AsyncCounter=0) -> (Output_Get=0)
   > ProcessBackpackAsync()
   ProcessHeadAsync() {is=title K2Node_CustomEvent}
   n40 = Get HeadTextures()
   > n41 = Branch(Condition=Is Array Non Empty!(InputArray:Array of Texture Object References=@n40) -> (Non-Empty:bool=~) {is=title K2Node_MacroInstance})
   n42 = Get AsyncCounter() -> (AsyncCounter:int=0)
   > n43 = Branch(Condition="integer >= integer"(A=@n42, B=Get TextureProcessingSize() -> (TextureProcessingSize:int=0)))
   ProcessBody(execute=[@n43.then, @n41.else]) {is=title K2Node_CallFunction}
   n44 = Get ConversionGraph() -> (ConversionGraph:Texture Graph Object Reference)
   SetTextureParameterValue!(
    execute=@n43.else
    WorldContextObject:Object Reference
    TextureGraph:Texture Graph Object Reference=@n44
    ParameterName=Base
    ParameterValue:Texture Object Reference="Get (a copy)"(Array:Array of Texture Object References=@n40, Dimension 1=@n42) -> (Output:Texture Object Reference)
   ) -> (then)
   > SetVectorParameterValue!(
    execute
    WorldContextObject:Object Reference
    TextureGraph:Texture Graph Object Reference=@n44
    ParameterName=Mixer
    ParameterValue_X:double="To Float (Boolean)"(InBool=Get OverrideSpecular?() -> (OverrideSpecular?:bool))
    ParameterValue_Y:double="To Float (Boolean)"(InBool=Get OverrideMetallic?() -> (OverrideMetallic?:bool))
    ParameterValue_Z:double="To Float (Boolean)"(InBool=Get OverrideRoughness?() -> (OverrideRoughness?:bool))
    ParameterValue_W:double=0
   ) -> (then)
   > SetVectorParameterValue(
    TextureGraph=@n44
    ParameterName=ChannelInput
    ParameterValue_X=Get SpecSliderValue() -> (SpecSliderValue:float)
    ParameterValue_Y=Get MetallicSliderValue() -> (MetallicSliderValue:float)
    ParameterValue_Z=Get RoughnessSliderValue() -> (RoughnessSliderValue:float)
    ParameterValue_W=1
   )
   > n45 = "Texture Graph Render (Async)"!(execute:Exec, InTextureGraph:Texture Graph Base Object Reference=@n44) -> (then:Exec, OnDone:Exec, OutputRts:Array of Texture Render Target 2D Object References) {is=title K2Node_AsyncAction}
   Set Array Elem(
    execute=@n45.OnDone
    TargetArray=Get ModHeadTextures()
    Index=@n42
    Item="Get (a copy)"(Array:Array of Texture Render Target 2D Object References=@n45) -> (Output:Texture Render Target 2D Object Reference)
    bSizeToFit=false
   )
   > Increment Int(Value=Get AsyncCounter() -> (AsyncCounter=0))
   > ProcessHeadAsync() {is=title K2Node_CallFunction}
   ProcessBodyAsync() {is=title K2Node_CustomEvent}
   n46 = Get BodyTextures()
   > n47 = Branch(Condition=Is Array Non Empty(InputArray=@n46) -> (Non-Empty=~))
   n48 = Get AsyncCounter() -> (AsyncCounter=0)
   > n49 = Branch(Condition="integer >= integer"(A=@n48, B=Get TextureProcessingSize() -> (TextureProcessingSize=0)))
   ProcessBackpack(execute=[@n47.else, @n49.then]) {is=title K2Node_CallFunction}
   n50 = Get ConversionGraph()
   SetTextureParameterValue(
    execute=@n49.else
    TextureGraph=@n50
    ParameterName=Base
    ParameterValue="Get (a copy)"(Array:Array of Texture Object References=@n46, Dimension 1=@n48) -> (Output:Texture Object Reference)
   )
   > SetVectorParameterValue(
    TextureGraph=@n50
    ParameterName=Mixer
    ParameterValue_X="To Float (Boolean)"(InBool=Get OverrideSpecular?())
    ParameterValue_Y="To Float (Boolean)"(InBool=Get OverrideMetallic?())
    ParameterValue_Z="To Float (Boolean)"(InBool=Get OverrideRoughness?())
    ParameterValue_W=0
   )
   > SetVectorParameterValue(
    TextureGraph=@n50
    ParameterName=ChannelInput
    ParameterValue_X=Get SpecSliderValue()
    ParameterValue_Y=Get MetallicSliderValue()
    ParameterValue_Z=Get RoughnessSliderValue()
    ParameterValue_W=1
   )
   > n51 = "Texture Graph Render (Async)"(InTextureGraph=@n50)
   Set Array Elem(
    execute=@n51.OnDone
    TargetArray=Get ModBodyTextures()
    Index=@n48
    Item="Get (a copy)"(Array:Array of Texture Render Target 2D Object References=@n51) -> (Output:Texture Render Target 2D Object Reference)
    bSizeToFit=false
   )
   > Increment Int(Value=Get AsyncCounter() -> (AsyncCounter=0))
   > ProcessBodyAsync() {is=title K2Node_CallFunction}
   ProcessBackpackAsync() {is=title K2Node_CustomEvent}
   n52 = Get BackPackTextures()
   > n53 = Branch(Condition=Is Array Non Empty(InputArray=@n52) -> (Non-Empty=~))
   n54 = Get AsyncCounter() -> (AsyncCounter=0)
   > n55 = Branch(Condition="integer >= integer"(A=@n54, B=Get TextureProcessingSize() -> (TextureProcessingSize=0)))
   ApplyTexture(execute=[@n53.else, @n55.then]) {is=title K2Node_CallFunction}
   n56 = Get ConversionGraph()
   SetTextureParameterValue(
    execute=@n55.else
    TextureGraph=@n56
    ParameterName=Base
    ParameterValue="Get (a copy)"(Array:Array of Texture Object References=@n52, Dimension 1=@n54) -> (Output:Texture Object Reference)
   )
   > SetVectorParameterValue(
    TextureGraph=@n56
    ParameterName=Mixer
    ParameterValue_X="To Float (Boolean)"(InBool=Get OverrideSpecular?())
    ParameterValue_Y="To Float (Boolean)"(InBool=Get OverrideMetallic?())
    ParameterValue_Z="To Float (Boolean)"(InBool=Get OverrideRoughness?())
    ParameterValue_W=0
   )
   > SetVectorParameterValue(
    TextureGraph=@n56
    ParameterName=ChannelInput
    ParameterValue_X=Get SpecSliderValue()
    ParameterValue_Y=Get MetallicSliderValue()
    ParameterValue_Z=Get RoughnessSliderValue()
    ParameterValue_W=1
   )
   > n57 = "Texture Graph Render (Async)"(InTextureGraph=@n56)
   Set Array Elem(
    execute=@n57.OnDone
    TargetArray=Get ModBackpackTextures()
    Index=@n54
    Item="Get (a copy)"(Array:Array of Texture Render Target 2D Object References=@n57) -> (Output:Texture Render Target 2D Object Reference)
    bSizeToFit=false
   )
   > Increment Int(Value=Get AsyncCounter() -> (AsyncCounter=0))
   > ProcessBackpackAsync() {is=title K2Node_CallFunction}
   RestoreToOGMaterials() {is=title K2Node_CustomEvent}
   > n58 = GetPawn()
   > n59 = Sequence()
   > n60 = GetTexturesForPartType(Target=@n58)
   > n61 = For Each Loop(Array:Array of Material Interface Object References=Get OGBodyMaterials() -> (OGBodyMaterials:Array of Material Interface Object References)) -> (Array Element:Material Interface Object Reference)
   > SetMaterial(self=@n60.PartSkelMesh, ElementIndex=@n61.Array Index, Material=@n61)
   n62 = GetTexturesForPartType(execute=@n59.then_1, Target=@n58, PartType=Head)
   > n63 = For Each Loop(Array=Get OGHeadMaterials() -> (OGHeadMaterials:Array of Material Interface Object References))
   > SetMaterial(self=@n62.PartSkelMesh, ElementIndex=@n63.Array Index, Material=@n63)
   n64 = GetTexturesForPartType(execute=@n59.then_2, Target=@n58, PartType=Backpack)
   > n65 = For Each Loop(Array=Get OGBackpackMaterials() -> (OGBackpackMaterials:Array of Material Interface Object References))
   > SetMaterial(self=@n64.PartSkelMesh, ElementIndex=@n65.Array Index, Material=@n65)
   n66 = Cast To FortPlayerMannequin!(execute, Object:Wildcard) -> (then, CastFailed, AsFort Player Mannequin:Fort Player Mannequin Object Reference) {is=title K2Node_DynamicCast}
   > n67 = For Each Loop(Array=GetMaterials!(self:Mesh Component Object Reference=Get SkeletalMeshComponent!(self:Skeletal Mesh Actor Object Reference=@n66) -> (SkeletalMeshComponent:Skeletal Mesh Component Object Reference)) -> (ReturnValue:Array of Material Interface Object References))
   > PrintString(InString=GetDisplayName(Object=@n67))
   Event Construct()
   > SetVisibility(self=Get AdvancedSliders(), InVisibility=Collapsed)
   > SetIsEnabled(self=[Get SpecSlider(), Get MetalSlider(), Get RoughSlider()], bInIsEnabled=false)
   ProcessHead() {is=title K2Node_CallFunction}
   RenderTextureGraph!(execute, WorldContextObject:Object Reference, InTextureGraph:Texture Graph Base Object Reference) -> (then, ReturnValue:Array of Texture Render Target 2D Object References)
   Event PreConstruct()
   Event Tick()
   Make Fort Pawn Material Override Copied Parameters(TextureParamNames=Make Array!("[0]"=TechArtMask, "[1]":Name=SkinFX_Mask) -> (Array))
  ]}
 GenerateAndApplyCustomUVIslandTextures {nodes [
   n1 = GenerateAndApplyCustomUVIslandTextures!() -> (
    then
    SkelMeshRef:Skeletal Mesh Component Object Reference
    MaterialsForPartType:Array of Material Interface Object References
    WorldContextObject:Object Reference
   ) {is=title K2Node_FunctionEntry}
   n2 = Get SkeletalMesh!(self:Skinned Mesh Component Object Reference=@n1) -> (SkeletalMesh:Skeletal Mesh Object Reference)
   > PrintString(InString=GetDisplayName(Object=@n2))
   > n3 = For Each Loop(Array=@n1.MaterialsForPartType)
   > n4 = SpawnTransientActor!(
    execute
    WorldContextObject:Object Reference=@n1.WorldContextObject
    ActorClass:Actor Class Reference=&/GDMABP_UVIslandViewer.GDMABP_UVIslandViewer_C
    SpawnTransform_Location:Vector="0, 0, 0"
    SpawnTransform_Rotation:Rotator="0, 0, 0"
    SpawnTransform_Scale:Vector="1,1,1"
    CollisionHandlingOverride:ESpawnActorCollisionHandlingMethod Enum=Undefined
    OwnerActor:Actor Object Reference
    InstigatorPawn:Pawn Object Reference
   ) -> (then, SpawnedActor:GDMABP UVIsland Viewer Object Reference)
   > n5 = RequestTextureForMaterialSlot!(
    execute
    self:GDMABP UVIsland Viewer Object Reference=@n4
    SkelMesh:Skeletal Mesh Object Reference=@n2
    MaterialSlotIndex:int=@n3.Array Index
   ) -> (then, OutputTexture:Texture 2D Object Reference)
   > PrintString(InString=GetDisplayName(Object=@n5))
   > n6 = CreateDynamicMaterialInstance(Parent=&/M_DisplayManualTexture, OptionalName=None, CreationFlags=None)
   > SetTextureParameterValue!(execute, self:Material Instance Dynamic Object Reference=@n6, ParameterName=Texture, Value:Texture Object Reference=@n5) -> (then)
   > SetMaterial(self=@n1, ElementIndex=@n3.Array Index, Material=@n6)
   > Destroy Actor(self=@n4)
  ]}
 GenerateRenderTarget {nodes [
   n1 = GenerateRenderTarget!() -> (then, BaseTexture:Texture Object Reference) {is=title K2Node_FunctionEntry}
   n2 = Get ConversionGraph()
   > SetTextureParameterValue!(
    execute
    WorldContextObject:Object Reference
    TextureGraph:Texture Graph Object Reference=@n2
    ParameterName=Base
    ParameterValue:Texture Object Reference=@n1
   ) -> (then)
   > SetVectorParameterValue(
    TextureGraph=@n2
    ParameterName=Mixer
    ParameterValue_X="To Float (Boolean)"(InBool=Get OverrideSpecular?())
    ParameterValue_Y="To Float (Boolean)"(InBool=Get OverrideMetallic?())
    ParameterValue_Z="To Float (Boolean)"(InBool=Get OverrideRoughness?())
    ParameterValue_W=0
   )
   > SetVectorParameterValue(
    TextureGraph=@n2
    ParameterName=ChannelInput
    ParameterValue_X=Get SpecSliderValue()
    ParameterValue_Y=Get MetallicSliderValue()
    ParameterValue_Z=Get RoughnessSliderValue()
    ParameterValue_W=1
   )
   > n3 = RenderTextureGraph(InTextureGraph=@n2)
   > Return Node!(execute, NewParam:Texture Object Reference="Get (a copy)"(Array=@n3))
   Get ConversionGraph()
  ]}
 GetPawn {nodes [
   GetPawn() {is=title K2Node_FunctionEntry}
   > n1 = GetSelectedActors()
   > n2 = Cast To FortPlayerPawn(Object="Get (a copy)"(Array:Array of Actor Object References=@n1) -> (Output:Actor Object Reference))
   > n3 = Set As Fort Player Pawn(As Fort Player Pawn:Fort Player Pawn Object Reference=@n2) -> (Output_Get:Fort Player Pawn Object Reference)
   > Return Node!(execute, Output_Get:Fort Player Pawn Object Reference=@n3)
   n4 = GetPIEWorlds(execute=@n2.CastFailed, bIncludeDedicatedServer=false)
   > n5 = GetAllActorsOfClass!(
    execute
    WorldContextObject:Object Reference="Get (a copy)"(Array:Array of World Object References=@n4) -> (Output:World Object Reference)
    ActorClass:Actor Class Reference=/Script/FortniteGame.FortPlayerPawn
   ) -> (then, OutActors:Array of Fort Player Pawn Object References)
   > n6 = Cast To FortPlayerPawn(Object="Get (a copy)"(Array:Array of Fort Player Pawn Object References=@n5) -> (Output:Fort Player Pawn Object Reference))
   > n7 = Set As Fort Player Pawn(As Fort Player Pawn=@n6)
   > Return Node(Output_Get=@n7)
  ]}
 GetTexturesForPartType {nodes [
   n1 = GetTexturesForPartType!() -> (then, Target:Fort Player Pawn Object Reference, PartType:EFortCustomPartType Enum=Head) {is=title K2Node_FunctionEntry}
   n2 = GetSkeletalMeshForPartType(self=@n1, PartType=@n1.PartType)
   > n3 = For Each Loop(Array:Array of Names=GetMaterialSlotNames(self=@n2)) -> (Array Element:Name)
   Return Node!(
    execute=@n3.Completed
    TexturesViaSlot1:Array of Texture Object References=Get TexturesViaSlot() -> (TexturesViaSlot:Array of Texture Object References)
    PartSkelMesh:Skeletal Mesh Component Object Reference=@n2
    MaterialsForPartType:Array of Material Interface Object References=GetMaterials(self=@n2)
   )
   > Set FoundPartTexture(FoundPartTexture:Texture Object Reference) -> (Output_Get:Texture Object Reference)
   > n4 = Get Dynamic Material Instance(SkeletalMeshComponent=@n2, ElementIndex=@n3.Array Index)
   > n5 = Get Texture Parameter Value!(execute, self:Material Instance Dynamic Object Reference=@n4, ParameterName:Name=SRM) -> (then, ReturnValue:Texture Object Reference) {is=title K2Node_CallFunction}
   > n6 = Is Valid(InputObject=@n5)
   n7 = Get Texture Parameter Value(execute=@n6.Is Not Valid, self=@n4, ParameterName=SpecularMasks)
   > n8 = Is Valid(InputObject=@n7)
   Add(execute=@n8.Is Not Valid, TargetArray:Array of Texture Object References=Get TexturesViaSlot(), NewItem:"Texture Object Reference (by ref)") {is=title K2Node_CallArrayFunction}
   n9 = Set FoundPartTexture(execute=@n8.Is Valid, FoundPartTexture=@n7)
   Add(
    execute=[@n9.then, @n10.then]
    TargetArray=Get TexturesViaSlot()
    NewItem=Get FoundPartTexture() -> (FoundPartTexture:Texture Object Reference)
   )
   n10 = Set FoundPartTexture(execute=@n6.Is Valid, FoundPartTexture=@n5)
  ]}
 OverrideTexture {nodes [OverrideTexture() {is=title K2Node_FunctionEntry}]}
 RestoreTexture {nodes [RestoreTexture() {is=title K2Node_FunctionEntry}]}
 SetTextureOverride_ApplyOverride {nodes [
   n1 = SetTextureOverride_ApplyOverride!() -> (then, Input:Material Interface Object Reference, MaterialParamsToCopy:"Fort Pawn Material Override Copied Parameters Structure (by ref)") {is=title K2Node_FunctionEntry}
   > n2 = GetPawn() {is=title K2Node_CallFunction}
   > n3 = Branch(Condition=Get IsDoingManualMaterialSwap?())
   > RestoreToOGMaterials() {is=title K2Node_CallFunction}
   > n4 = Set IsDoingManualMaterialSwap?(IsDoingManualMaterialSwap?=false)
   RemoveMaterialOverrideLocal(execute=[@n3.else, @n4.then], self=@n2, MaterialOverrideId=Get MaterialOverride())
   > n5 = Apply Material Override Local with Flags(self=@n2, Material=Make Soft Reference(Input=@n1), MaterialParamsToCopy=@n1.MaterialParamsToCopy, Priority=1, Flags=15)
   > Set MaterialOverride(MaterialOverride=@n5)
   Get Material Params to Copy()
  ]}
]
variables [
 As Fort Player Pawn {type=FortPlayerPawn}
 AsyncCounter {type=integer}
 BackPackTextures {type="array<Texture>", default []}
 BodyTextures {type="array<Texture>", default []}
 ConversionGraph {type=TextureGraph, default=&/TG_Artreview_SRMReplace}
 FXMaskSettings {
  type=FortPawnMaterialOverrideCopiedParameters
  default {scalarParamNames [], textureParamNames [TechArtMask,SkinFX_Mask], vectorParamNames []}
 }
 GenerateOverride {type=TextureRenderTarget2D, default=null}
 HeadTextures {type="array<Texture>", default []}
 HoloMaterialInst {type=MaterialInterface, default=/Game/NeverCook/ArtTools/Tech_Art/ArtReview_Visualizer/PostProcess_Mats/CharacterOverrides/M_DebugNormals}
 IsDoingManualMaterialSwap? {type=boolean}
 Material Params to Copy {type=FortPawnMaterialOverrideCopiedParameters, default {scalarParamNames [], textureParamNames [normals], vectorParamNames []}}
 MaterialOverride {type=Guid, default=#1}
 MetallicSliderValue {type=number, default=0.5}
 MMaskSettings {type=FortPawnMaterialOverrideCopiedParameters, default {scalarParamNames [], textureParamNames [M,Mask], vectorParamNames []}}
 ModBackpackTextures {type="array<Texture>", default []}
 ModBodyTextures {type="array<Texture>", default []}
 ModHeadTextures {type="array<Texture>", default []}
 OGBackpackMaterials {type="array<MaterialInterface>", default []}
 OGBodyMaterials {type="array<MaterialInterface>", default []}
 OGHeadMaterials {type="array<MaterialInterface>", default []}
 OverrideMetallic? {type=boolean}
 OverrideRoughness? {type=boolean}
 OverrideSpecular? {type=boolean}
 Poly Group IDs Out {type=GeometryScriptIndexList, default {indexType=Any}}
 RoughnessSliderValue {type=number, default=0.5}
 SkeletalMeshActor {type=SkeletalMesh, default=/BRCosmetics/Characters/Player/Male/Medium/Bodies/M_MED_Orbit_Teal/Meshes/M_MED_Orbit_Teal}
 SpecSliderValue {type=number, default=0.5}
 SRMSettings {type=FortPawnMaterialOverrideCopiedParameters, default {scalarParamNames [], textureParamNames [SpecularMasks, SRM], vectorParamNames []}}
 TextureProcessingSize {type=integer}
]
widgets {
 info {rootWidgetClass=&Overlay, widgetCount=43}
 widgets [
  Overlay_17 {widgetClassPath=&Overlay, widget=Overlay_17}
  Border_38 {
   widgetClassPath=&Border
   parent=Overlay_17
   widget=Border_38
   properties {verticalAlignment=VAlign_Top, brushColor=<rgba 0.1562 0.1562 0.1562 1>}
   slot {class=OverlaySlot, properties {horizontalAlignment=HAlign_Fill}}
  }
  VerticalBox_0 {
   widgetClassPath=&VerticalBox
   parent=Border_38
   widget=VerticalBox_0
   slot {class=BorderSlot, properties {verticalAlignment=VAlign_Top}}
  }
  HorizontalBox_1 {widgetClassPath=&HorizontalBox, parent=VerticalBox_0, widget=HorizontalBox_1, slot {class=VerticalBoxSlot}}
  HorizontalBox_35 {
   widgetClassPath=&HorizontalBox
   parent=HorizontalBox_1
   widget=HorizontalBox_35
   slot {class=HorizontalBoxSlot, properties {size {sizeRule=Fill, value=1}}}
  }
  TextBlock_111 {
   widgetClassPath=&TextBlock
   parent=HorizontalBox_35
   widget=TextBlock_111
   properties {
    text=Character Texture Debugs
    colorAndOpacity {colorUseRule=UseColor_Specified, specifiedColor=<rgba 0 0 0 1>}
    font {
     bForceMonospaced=false
     bMaterialIsStencil=false
     fontMaterial=null
     fontObject=&Roboto
     letterSpacing=0
     monospacedWidth=1
     size=12
     skewAmount=0
     typefaceFontName=Bold
     outlineSettings {
      bApplyOutlineToDropShadows=false
      bMiteredCorners=false
      bSeparateFillAlpha=false
      outlineMaterial=null
      outlineSize=0
      outlineColor=<rgba 0 0 0 1>
     }
    }
   }
   slot {class=HorizontalBoxSlot, properties {verticalAlignment=VAlign_Center}}
  }
  WrapBox_74 {
   widgetClassPath=&WrapBox
   parent=HorizontalBox_35
   widget=WrapBox_74
   slot {class=HorizontalBoxSlot, properties {size {sizeRule=Fill, value=1}}}
  }
  Button_Restore {
   widgetClassPath=&TA_Generic01_Button
   parent=WrapBox_74
   widget=Button_Restore
   properties {label=Restore}
   slot {class=WrapBoxSlot}
  }
  Button_Diffuse {
   widgetClassPath=&TA_Generic01_Button
   parent=WrapBox_74
   widget=Button_Diffuse
   properties {label=Diffuse}
   slot {class=WrapBoxSlot}
  }
  Button_Normals {
   widgetClassPath=&TA_Generic01_Button
   parent=WrapBox_74
   widget=Button_Normals
   properties {label=Normals}
   slot {class=WrapBoxSlot}
  }
  Button_FXMask {
   widgetClassPath=&TA_Generic01_Button
   parent=WrapBox_74
   widget=Button_FXMask
   properties {label=SkinFX_Mask/TechArtMask}
   slot {class=WrapBoxSlot}
  }
  Button_UVCheck {
   widgetClassPath=&TA_Generic01_Button
   parent=WrapBox_74
   widget=Button_UVCheck
   properties {
    label=UV Checker
    buttonColor_Default=<rgba 1 0.9885 0 1>
    foregroundColor {colorUseRule=UseColor_Specified, specifiedColor=<rgba 1 0 1 1>}
    labelColor_Default=<rgba 0 0 0 1>
    labelColor_Selected=<rgba 1 1 1 1>
   }
   slot {class=WrapBoxSlot}
  }
  Button_UVIslands {
   widgetClassPath=&TA_Generic01_Button
   parent=WrapBox_74
   widget=Button_UVIslands
   properties {
    label=UVIslands
    toolTipText=Islands that are designed to overlap will be colored in RED.
    buttonColor_Default=<rgba 1 0.4768 0 1>
    foregroundColor {colorUseRule=UseColor_Specified, specifiedColor=<rgba 1 0 1 1>}
    labelColor_Default=<rgba 0 0 0 1>
    labelColor_Selected=<rgba 1 1 1 1>
   }
   slot {class=WrapBoxSlot}
  }
  Spacer_163 {widgetClassPath=&Spacer, parent=WrapBox_74, widget=Spacer_163, slot {class=WrapBoxSlot, properties {padding=<ltrb 5 5 5 5>}}}
  Button_MMask {
   widgetClassPath=&TA_Generic01_Button
   parent=WrapBox_74
   widget=Button_MMask
   properties {label=M/Mask, buttonColor_Default=<rgba 1 1 1 1>, labelColor_Default=<rgba 0 0 0 1>}
   slot {class=WrapBoxSlot}
  }
  Button_MMask_R {
   widgetClassPath=&TA_Generic01_Button
   parent=WrapBox_74
   widget=Button_MMask_R
   properties {label=Ao, buttonColor_Default=<rgba 1 0 0 1>, labelColor_Default=<rgba 1 1 1 1>}
   slot {class=WrapBoxSlot}
  }
  Button_MMask_G {
   widgetClassPath=&TA_Generic01_Button
   parent=WrapBox_74
   widget=Button_MMask_G
   properties {label=Gmap, buttonColor_Default=<rgba 0 1 0 1>, labelColor_Default=<rgba 0 0 0 1>}
   slot {class=WrapBoxSlot}
  }
  Button_MMask_B {
   widgetClassPath=&TA_Generic01_Button
   parent=WrapBox_74
   widget=Button_MMask_B
   properties {label=Skin, buttonColor_Default=<rgba 0 0 1 1>, labelColor_Default=<rgba 1 1 1 1>}
   slot {class=WrapBoxSlot}
  }
  Spacer_285 {widgetClassPath=&Spacer, parent=WrapBox_74, widget=Spacer_285, slot {class=WrapBoxSlot, properties {padding=<ltrb 5 5 5 5>}}}
  Button_SRM {
   widgetClassPath=&TA_Generic01_Button
   parent=WrapBox_74
   widget=Button_SRM
   properties {label=SRM/SpecularMask, buttonColor_Default=<rgba 1 1 1 1>, labelColor_Default=<rgba 0 0 0 1>}
   slot {class=WrapBoxSlot}
  }
  Button_SRM_R {
   widgetClassPath=&TA_Generic01_Button
   parent=WrapBox_74
   widget=Button_SRM_R
   properties {label=Specular, buttonColor_Default=<rgba 1 0 0 1>}
   slot {class=WrapBoxSlot}
  }
  Button_SRM_G {
   widgetClassPath=&TA_Generic01_Button
   parent=WrapBox_74
   widget=Button_SRM_G
   properties {label=Metallic, buttonColor_Default=<rgba 0 1 0 1>, labelColor_Default=<rgba 0 0 0 1>}
   slot {class=WrapBoxSlot}
  }
  Button_SRM_B {
   widgetClassPath=&TA_Generic01_Button
   parent=WrapBox_74
   widget=Button_SRM_B
   properties {label=Roughness, buttonColor_Default=<rgba 0 0 1 1>}
   slot {class=WrapBoxSlot}
  }
  Button_Fuzz {
   widgetClassPath=&TA_Generic01_Button
   parent=WrapBox_74
   widget=Button_Fuzz
   properties {isEnabled=false, label=ClothFuzz, visibility=Hidden}
   slot {class=WrapBoxSlot}
  }
  AdvancedMaps {
   widgetClassPath=&EditorUtilityCheckBox
   bIsVariable=true
   parent=HorizontalBox_35
   widget=AdvancedMaps
   slot {class=HorizontalBoxSlot, properties {horizontalAlignment=HAlign_Left, padding=<ltrb 0 0 100 0>}}
  }
  TextBlock_118 {
   widgetClassPath=&TextBlock
   parent=AdvancedMaps
   widget=TextBlock_118
   properties {
    text="Use Single Channel Override (BETA!)"
    font {
     bForceMonospaced=false
     bMaterialIsStencil=false
     fontMaterial=null
     fontObject=&Roboto
     letterSpacing=0
     monospacedWidth=1
     size=14.25
     skewAmount=0
     typefaceFontName=Bold
     outlineSettings {
      bApplyOutlineToDropShadows=false
      bMiteredCorners=false
      bSeparateFillAlpha=false
      outlineMaterial=null
      outlineSize=0
      outlineColor=<rgba 0 0 0 1>
     }
    }
   }
   slot {class=PanelSlot}
  }
  HorizontalBox_2 {
   widgetClassPath=&HorizontalBox
   parent=VerticalBox_0
   widget=HorizontalBox_2
   slot {class=VerticalBoxSlot, properties {size {sizeRule=Fill, value=0.2}}}
  }
  AdvancedSliders {
   widgetClassPath=&HorizontalBox
   bIsVariable=true
   parent=HorizontalBox_2
   widget=AdvancedSliders
   slot {class=HorizontalBoxSlot, properties {size {sizeRule=Fill, value=1}}}
  }
  SpecularBox {
   widgetClassPath=&HorizontalBox
   parent=AdvancedSliders
   widget=SpecularBox
   slot {class=HorizontalBoxSlot, properties {padding=<ltrb 0 0 50 0>, size {sizeRule=Fill, value=1}}}
  }
  Specular {
   widgetClassPath=&EditorUtilityCheckBox
   bIsVariable=true
   parent=SpecularBox
   widget=Specular
   slot {class=HorizontalBoxSlot}
  }
  TextBlock_223 {
   widgetClassPath=&TextBlock
   parent=Specular
   widget=TextBlock_223
   properties {
    text=Spec Override
    font {
     bForceMonospaced=false
     bMaterialIsStencil=false
     fontMaterial=null
     fontObject=&Roboto
     letterSpacing=0
     monospacedWidth=1
     size=11.25
     skewAmount=0
     typefaceFontName=Bold
     outlineSettings {
      bApplyOutlineToDropShadows=false
      bMiteredCorners=false
      bSeparateFillAlpha=false
      outlineMaterial=null
      outlineSize=0
      outlineColor=<rgba 0 0 0 1>
     }
    }
   }
   slot {class=PanelSlot}
  }
  SpecValue {
   widgetClassPath=&TextBlock
   bIsVariable=true
   parent=SpecularBox
   widget=SpecValue
   properties {
    text=0.5
    font {
     bForceMonospaced=false
     bMaterialIsStencil=false
     fontMaterial=null
     fontObject=&Roboto
     letterSpacing=0
     monospacedWidth=1
     size=14.25
     skewAmount=0
     typefaceFontName=Bold
     outlineSettings {
      bApplyOutlineToDropShadows=false
      bMiteredCorners=false
      bSeparateFillAlpha=false
      outlineMaterial=null
      outlineSize=0
      outlineColor=<rgba 0 0 0 1>
     }
    }
   }
   slot {class=HorizontalBoxSlot, properties {padding=<ltrb 2 0 0 0>}}
  }
  SpecSlider {
   widgetClassPath=/Script/Blutility.EditorUtilitySlider
   bIsVariable=true
   parent=SpecularBox
   widget=SpecSlider
   properties {value=0.5}
   slot {class=HorizontalBoxSlot, properties {size {sizeRule=Fill, value=1}}}
  }
  MetallicBox {
   widgetClassPath=&HorizontalBox
   parent=AdvancedSliders
   widget=MetallicBox
   slot {class=HorizontalBoxSlot, properties {padding=<ltrb 0 0 50 0>, size {sizeRule=Fill, value=1}}}
  }
  Metallic {
   widgetClassPath=&EditorUtilityCheckBox
   bIsVariable=true
   parent=MetallicBox
   widget=Metallic
   slot {class=HorizontalBoxSlot}
  }
  TextBlock {
   widgetClassPath=&TextBlock
   parent=Metallic
   widget=TextBlock
   properties {
    text=Metal Override
    font {
     bForceMonospaced=false
     bMaterialIsStencil=false
     fontMaterial=null
     fontObject=&Roboto
     letterSpacing=0
     monospacedWidth=1
     size=11.25
     skewAmount=0
     typefaceFontName=Bold
     outlineSettings {
      bApplyOutlineToDropShadows=false
      bMiteredCorners=false
      bSeparateFillAlpha=false
      outlineMaterial=null
      outlineSize=0
      outlineColor=<rgba 0 0 0 1>
     }
    }
   }
   slot {class=PanelSlot}
  }
  MetalValue {
   widgetClassPath=&TextBlock
   bIsVariable=true
   parent=MetallicBox
   widget=MetalValue
   properties {
    text=0.5
    font {
     bForceMonospaced=false
     bMaterialIsStencil=false
     fontMaterial=null
     fontObject=&Roboto
     letterSpacing=0
     monospacedWidth=1
     size=14.25
     skewAmount=0
     typefaceFontName=Bold
     outlineSettings {
      bApplyOutlineToDropShadows=false
      bMiteredCorners=false
      bSeparateFillAlpha=false
      outlineMaterial=null
      outlineSize=0
      outlineColor=<rgba 0 0 0 1>
     }
    }
   }
   slot {class=HorizontalBoxSlot, properties {padding=<ltrb 2 0 0 0>}}
  }
  MetalSlider {
   widgetClassPath=&EditorUtilitySlider
   bIsVariable=true
   parent=MetallicBox
   widget=MetalSlider
   properties {value=0.5}
   slot {class=HorizontalBoxSlot, properties {size {sizeRule=Fill, value=1}}}
  }
  RoughnessBox {
   widgetClassPath=&HorizontalBox
   parent=AdvancedSliders
   widget=RoughnessBox
   slot {class=HorizontalBoxSlot, properties {padding=<ltrb 0 0 50 0>, size {sizeRule=Fill, value=1}}}
  }
  Roughness {
   widgetClassPath=&EditorUtilityCheckBox
   bIsVariable=true
   parent=RoughnessBox
   widget=Roughness
   slot {class=HorizontalBoxSlot}
  }
  TextBlock_1 {
   widgetClassPath=&TextBlock
   parent=Roughness
   widget=TextBlock_1
   properties {
    text=Roughness Override
    font {
     bForceMonospaced=false
     bMaterialIsStencil=false
     fontMaterial=null
     fontObject=&Roboto
     letterSpacing=0
     monospacedWidth=1
     size=11.25
     skewAmount=0
     typefaceFontName=Bold
     outlineSettings {
      bApplyOutlineToDropShadows=false
      bMiteredCorners=false
      bSeparateFillAlpha=false
      outlineMaterial=null
      outlineSize=0
      outlineColor=<rgba 0 0 0 1>
     }
    }
   }
   slot {class=PanelSlot}
  }
  RoughValue {
   widgetClassPath=&TextBlock
   bIsVariable=true
   parent=RoughnessBox
   widget=RoughValue
   properties {
    text=0.5
    font {
     bForceMonospaced=false
     bMaterialIsStencil=false
     fontMaterial=null
     fontObject=&Roboto
     letterSpacing=0
     monospacedWidth=1
     size=14.25
     skewAmount=0
     typefaceFontName=Bold
     outlineSettings {
      bApplyOutlineToDropShadows=false
      bMiteredCorners=false
      bSeparateFillAlpha=false
      outlineMaterial=null
      outlineSize=0
      outlineColor=<rgba 0 0 0 1>
     }
    }
   }
   slot {class=HorizontalBoxSlot, properties {padding=<ltrb 2 0 0 0>}}
  }
  RoughSlider {
   widgetClassPath=&EditorUtilitySlider
   bIsVariable=true
   parent=RoughnessBox
   widget=RoughSlider
   properties {value=0.5}
   slot {class=HorizontalBoxSlot, properties {size {sizeRule=Fill, value=1}}}
  }
 ]
}
