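# Art review debug material function that overrides or offsets Metallic, Roughness and Specular on incoming material attributes.
# Optional scalar inputs for each channel default to the values extracted from the attributes input.
# Per channel, an OverrideOrAdjust parameter blends between replacing and adding the debug value, and an Enable parameter, off by default, blends in that result.
# Outputs are the modified material attributes plus the three separate scalars.
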MaterialFunction /Game/NeverCook/ArtTools/Tech_Art/ArtReview_Visualizer/PostProcess_Mats/MF_ArtReview_Debugger
graph {nodes [
  n1 = GetMaterialAttributes(None=FunctionInput() {bUsePreviewValueAsDefault=true, inputType=FunctionInput_MaterialAttributes}) {attributeGetTypes [#1, #2, #3]}
  n2 = ScalarParameter() {defaultValue=0.5, parameterName=Metallic}
  n3 = FunctionInput(Preview=@n1.Metallic) {bUsePreviewValueAsDefault=true, inputName=Metallic, inputType=FunctionInput_Scalar, sortPriority=1}
  n4 = LinearInterpolate(
   A=@n3
   B=LinearInterpolate(A=@n2, B=Add(A=@n3, B=@n2), Alpha=ScalarParameter() {parameterName=Metal_OverrideOrAdjust}) {desc=Metallic}
   Alpha=ScalarParameter() {parameterName=Metal_Enable}
  ) {desc=Metallic}
  n5 = ScalarParameter() {defaultValue=0.5, parameterName=Roughness}
  n6 = FunctionInput(Preview=@n1.Roughness) {bUsePreviewValueAsDefault=true, inputName=Roughness, inputType=FunctionInput_Scalar, sortPriority=2}
  n7 = LinearInterpolate(
   A=@n6
   B=LinearInterpolate(A=@n5, B=Add(A=@n6, B=@n5), Alpha=ScalarParameter() {parameterName=Rough_OverrideOrAdjust}) {desc=Mettalic}
   Alpha=ScalarParameter() {parameterName=Rough_Enable}
  ) {desc=Mettalic}
  n8 = ScalarParameter() {defaultValue=0.5, parameterName=Specular}
  n9 = FunctionInput(Preview=@n1.Specular) {bUsePreviewValueAsDefault=true, inputName=Specular, inputType=FunctionInput_Scalar, sortPriority=3}
  n10 = LinearInterpolate(
   A=@n9
   B=LinearInterpolate(A=@n8, B=Add(A=@n9, B=@n8), Alpha=ScalarParameter() {parameterName=Spec_OverrideOrAdjust}) {desc=Mettalic}
   Alpha=ScalarParameter() {parameterName=Spec_Enable}
  ) {desc=Mettalic}
  FunctionOutput(None=SetMaterialAttributes!(MaterialAttributes=@n1.MaterialAttributes, Metallic=@n4, Roughness=@n7, Specular=@n10) {attributeSetTypes [#1, #2, #3]}) {outputName=MaterialAttributes}
  FunctionOutput(None=@n4) {outputName=Metallic}
  FunctionOutput(None=@n7) {outputName=Roughness}
  FunctionOutput(None=@n10) {outputName=Specular}
 ]}
