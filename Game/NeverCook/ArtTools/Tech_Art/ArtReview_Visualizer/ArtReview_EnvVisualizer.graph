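# Editor utility widget for art review that previews roughness, metallic and specular overrides or additive adjustments on a selected level actor.
# The button hides the actor and spawns a transient mesh stand-in; clicking again restores the original.
# Stand-in material parent chains are recursively duplicated to the base Material, where MF_ArtReview_Debugger is spliced in and recompiled.
# Checkboxes and sliders push scalar parameters to tracked dynamic instances of these copies.
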EditorUtilityWidgetBlueprint /Game/NeverCook/ArtTools/Tech_Art/ArtReview_Visualizer/ArtReview_EnvVisualizer
graphs [
 CheckIsValidForDebug {nodes [
   n1 = CheckIsValidForDebug!() -> (then, Context:Actor Object Reference) {is=title K2Node_FunctionEntry}
   > n2 = Gather Subobject Data For Instance(self=Get SubobjectDataSubsystem(), Context=@n1)
   > n3 = For Each Loop(Array:Array of Subobject Data Handle Structures=@n2) -> (Array Element:Subobject Data Handle Structure)
   Return Node!(execute=@n3.Completed, IsValid?:bool=Get HasMeshRenderComponents?() -> (HasMeshRenderComponents?:bool))
   n4 = GetAssociatedObject!(Data:"Subobject Data Structure (by ref)"=GetData(DataHandle=@n3)) -> (ReturnValue:Object Reference)
   > n5 = Cast To StaticMeshComponent(Object=@n4)
   Set HasMeshRenderComponents?(execute=[@n6.then, @n5.then], HasMeshRenderComponents?:bool=true) -> (Output_Get:bool)
   n6 = Cast To SkeletalMeshComponent!(execute=@n5.CastFailed, Object=@n4) -> (then, CastFailed, AsSkeletal Mesh Component:Skeletal Mesh Component Object Reference) {is=title K2Node_DynamicCast}
  ]}
 CompareArrays {nodes [CompareArrays() {is=title K2Node_FunctionEntry}]}
 CreateTransientStandin {nodes [
   n1 = CreateTransientStandin!() -> (then, ActorClass:Actor Object Reference) {is=title K2Node_FunctionEntry}
   n2 = Get Class(Object=@n1)
   > n3 = Branch(Condition=ClassIsChildOf(TestClass=@n2, ParentClass=/Script/Engine.SkeletalMeshActor))
   n4 = Get Actor Transform(self=@n1)
   > n5 = SpawnTransientActor!(
    execute
    WorldContextObject:Object Reference
    ActorClass:Actor Class Reference=@n2
    SpawnTransform:"Transform (by ref)"=@n4
    CollisionHandlingOverride:ESpawnActorCollisionHandlingMethod Enum=Undefined
    OwnerActor:Actor Object Reference
    InstigatorPawn:Pawn Object Reference
   ) -> (then, SpawnedActor:Actor Object Reference)
   > n6 = Cast To SkeletalMeshActor!(execute, Object=@n5) -> (then, CastFailed, AsSkeletal Mesh Actor:Skeletal Mesh Actor Object Reference) {is=title K2Node_DynamicCast}
   > n7 = Cast To SkeletalMeshActor(Object=@n1)
   n8 = Get SkeletalMeshComponent!(self:Skeletal Mesh Actor Object Reference=@n7) -> (SkeletalMeshComponent:Skeletal Mesh Component Object Reference)
   n9 = Get SkeletalMeshComponent(self=@n6)
   > SetSkeletalMeshAsset!(
    execute
    self:Skeletal Mesh Component Object Reference=@n9
    NewMesh:Skeletal Mesh Object Reference=GetSkeletalMeshAsset!(self:Skeletal Mesh Component Object Reference=@n8) -> (ReturnValue:Skeletal Mesh Object Reference)
   ) -> (then)
   > SetComponentMaterialList!(
    execute
    Component:Primitive Component Object Reference=@n9
    MaterialList:Array of Material Interface Object References=GetMaterials!(self:Mesh Component Object Reference=@n8) -> (ReturnValue:Array of Material Interface Object References)
   ) -> (then)
   > Return Node!(execute, OutputPin:Actor Object Reference=@n5)
   n10 = Branch(execute=@n3.else, Condition=ClassIsChildOf(TestClass=@n2, ParentClass=&StaticMeshActor))
   > n11 = SpawnTransientActor(ActorClass=@n2, SpawnTransform=@n4, CollisionHandlingOverride=Undefined)
   > n12 = Cast To StaticMeshActor!(execute, Object=@n11) -> (then, CastFailed, AsStatic Mesh Actor:Static Mesh Actor Object Reference) {is=title K2Node_DynamicCast}
   > n13 = Cast To StaticMeshActor(Object=@n1)
   n14 = Get StaticMeshComponent(self=@n13)
   n15 = Get StaticMeshComponent(self=@n12)
   > SetStaticMesh(self=@n15, NewMesh=Get StaticMesh(self=@n14))
   > SetComponentMaterialList(Component=@n15, MaterialList=GetMaterials(self=@n14))
   > Return Node(OutputPin=@n11)
   n16 = SpawnTransientActor(execute=@n10.else, ActorClass=&Actor, SpawnTransform=@n4, CollisionHandlingOverride=Undefined)
   > n17 = Gather Subobject Data For Instance(self=Get SubobjectDataSubsystem(), Context=@n1)
   > n18 = For Each Loop(Array=@n17)
   Return Node(execute=@n18.Completed, OutputPin=@n16)
   n19 = GetAssociatedObject(Data=GetData(DataHandle=@n18))
   > n20 = Cast To StaticMeshComponent(Object=@n19)
   > n21 = Add Component by Class!(
    execute:Exec
    self:Actor Object Reference=@n16
    WorldContextObject:Object Reference
    Class:Actor Component Class Reference=&StaticMeshComponent
    bManualAttachment:bool
    RelativeTransform_Location:Vector=Get RelativeLocation!(self:Scene Component Object Reference=@n20) -> (RelativeLocation:Vector)
    RelativeTransform_Rotation:Rotator=Get RelativeRotation!(self:Scene Component Object Reference=@n20) -> (RelativeRotation:Rotator)
    RelativeTransform_Scale:Vector=Get RelativeScale3D!(self:Scene Component Object Reference=@n20) -> (RelativeScale3D:Vector)
   ) -> (then:Exec, ReturnValue:Static Mesh Component Object Reference) {is=title K2Node_AddComponentByClass}
   > SetStaticMesh(self=@n21, NewMesh=Get StaticMesh(self=@n20))
   > SetComponentMaterialList(Component=@n21, MaterialList=GetMaterials(self=@n20))
   n22 = Cast To SkeletalMeshComponent(execute=@n20.CastFailed, Object=@n19)
   > n23 = Add Component by Class!(
    execute
    self=@n16
    WorldContextObject
    Class=/Script/Engine.SkeletalMeshComponent
    bManualAttachment
    RelativeTransform:Transform=GetRelativeTransform!(self:Scene Component Object Reference=@n22) -> (ReturnValue:Transform)
   ) -> (then, ReturnValue:Skeletal Mesh Component Object Reference)
   > SetSkeletalMeshAsset(
    self=@n23
    NewMesh=Get SkeletalMeshAsset!(self:Skeletal Mesh Component Object Reference=@n22) -> (SkeletalMeshAsset:Skeletal Mesh Object Reference)
   )
   > SetComponentMaterialList(Component=@n23, MaterialList=GetMaterials(self=@n22))
  ]}
 EventGraph {nodes [
   n1 = Get EditorActorSubsystem()
   n2 = GetSelectedLevelActors(execute=@n19.then, self=@n1)
   n3 = "Get (a copy)"(Array:Array of Actor Object References=@n2) -> (Output:Actor Object Reference)
   n4 = Set DebugTarget(execute=@n50.then, DebugTarget:Actor Object Reference=@n3) -> (Output_Get:Actor Object Reference)
   n5 = DuplicateActor!(
    execute
    self:Editor Actor Subsystem Object Reference=@n1
    ActorToDuplicate:Actor Object Reference=@n4
    ToWorld:World Object Reference
    Offset:Vector
   ) -> (then, ReturnValue:Actor Object Reference)
   n6 = Clear(
    execute=[@n5.then, @n7.then]
    TargetArray:Array of Material Instance Dynamic Object References=Get TrackedMaterialsForDebug() -> (TrackedMaterialsForDebug:Array of Material Instance Dynamic Object References)
   )
   n7 = CreateTransientStandin!(execute=@n48.then, self, ActorClass:Actor Object Reference=@n4) -> (then, OutputPin:Actor Object Reference) {is=title K2Node_CallFunction}
   Set TransientDuplicate(execute=@n6.then, TransientDuplicate:Actor Object Reference=@n7) -> (Output_Get:Actor Object Reference)
   > n8 = Gather Subobject Data For Instance(self=Get SubobjectDataSubsystem(), Context=@n7)
   > n9 = For Each Loop(Array=@n8)
   n10 = GetAssociatedObject(Data=GetData(DataHandle=@n9))
   > n11 = Cast To SkeletalMeshComponent(Object=@n10)
   > n12 = For Each Loop(Array:Array of Names=GetMaterialSlotNames!(self:Primitive Component Object Reference=@n11) -> (ReturnValue:Array of Names)) -> (Array Element:Name)
   > n13 = GetAllParentMIsToBaseM_Recursive!(
    execute
    self
    Material interface in:Material Interface Object Reference=GetMaterial!(self:Primitive Component Object Reference=@n11, ElementIndex:int=@n12.Array Index) -> (ReturnValue:Material Interface Object Reference)
   ) -> (then, InputTempVersion:Material Interface Object Reference)
   > n14 = CreateDynamicMaterialInstance!(
    execute
    self:Primitive Component Object Reference=@n11
    ElementIndex:int=@n12.Array Index
    SourceMaterial:Material Interface Object Reference=@n13
    OptionalName:Name=None
   ) -> (then, ReturnValue:Material Instance Dynamic Object Reference) {is=title K2Node_CallFunction}
   > Add(
    TargetArray:Array of Material Instance Dynamic Object References=Get TrackedMaterialsForDebug()
    NewItem:"Material Instance Dynamic Object Reference (by ref)"=@n14
   ) {is=title K2Node_CallArrayFunction}
   > SetMaterial(self=@n11, ElementIndex=@n12.Array Index, Material=@n14)
   n15 = Cast To StaticMeshComponent(execute=@n11.CastFailed, Object=@n10)
   > n16 = For Each Loop(Array=GetMaterialSlotNames(self=@n15))
   > n17 = GetAllParentMIsToBaseM_Recursive(Material interface in=GetMaterial(self=@n15, ElementIndex=@n16.Array Index))
   > n18 = CreateDynamicMaterialInstance(self=@n15, ElementIndex=@n16.Array Index, SourceMaterial=@n17, OptionalName=None)
   > Add(TargetArray=Get TrackedMaterialsForDebug(), NewItem=@n18)
   > SetMaterial(self=@n15, ElementIndex=@n16.Array Index, Material=@n18)
   n19 = "On Clicked (TA_Generic01_Button)"!() -> (OutputDelegate, then, widget:User Widget Object Reference, event:Pointer Event Structure) {is=title K2Node_ComponentBoundEvent}
   n20 = "On Check State Changed (Specular)"!() -> (OutputDelegate, then, bIsChecked:bool) {is=title K2Node_ComponentBoundEvent}
   > n21 = Branch(Condition=@n20.bIsChecked)
   > SendDebugValue!(execute, self, ParameterName:Name=Spec_Enable, Value:float=1) -> (then)
   SendDebugValue(execute=@n21.else, ParameterName=Spec_Enable, Value=0)
   n22 = "On Check State Changed (Metallic)"!() -> (OutputDelegate, then, bIsChecked:bool) {is=title K2Node_ComponentBoundEvent}
   > n23 = Branch(Condition=@n22.bIsChecked)
   > SendDebugValue(ParameterName=Metal_Enable, Value=1)
   SendDebugValue(execute=@n23.else, ParameterName=Metal_Enable, Value=0)
   n24 = "On Check State Changed (Roughness)"!() -> (OutputDelegate, then, bIsChecked:bool) {is=title K2Node_ComponentBoundEvent}
   > n25 = Branch(Condition=@n24.bIsChecked)
   > SendDebugValue(ParameterName=Rough_Enable, Value=1)
   SendDebugValue(execute=@n25.else, ParameterName=Rough_Enable, Value=0)
   n26 = "On Check State Changed (SpecMode)"!() -> (OutputDelegate, then, bIsChecked:bool) {is=title K2Node_ComponentBoundEvent}
   > n27 = Branch(Condition=@n26.bIsChecked)
   > SendDebugValue(ParameterName=Spec_OverrideOrAdjust, Value=1)
   n28 = Get SpecSlider() -> (SpecSlider:Editor Utility Slider Object Reference)
   > SetMinValue!(execute, self:Slider Object Reference=@n28, InValue:float=-1) -> (then)
   > SetMaxValue!(execute, self:Slider Object Reference=@n28, InValue:float=1) -> (then)
   SendDebugValue(execute=@n27.else, ParameterName=Spec_OverrideOrAdjust, Value=0)
   > SetMinValue(self=@n28, InValue=0)
   > SetMaxValue(self=@n28, InValue=1)
   n29 = "On Check State Changed (RoughMode)"!() -> (OutputDelegate, then, bIsChecked:bool) {is=title K2Node_ComponentBoundEvent}
   > n30 = Branch(Condition=@n29.bIsChecked)
   > SendDebugValue(ParameterName=Rough_OverrideOrAdjust, Value=1)
   n31 = Get RoughSlider() -> (RoughSlider:Editor Utility Slider Object Reference)
   > SetMinValue(self=@n31, InValue=-1)
   > SetMaxValue(self=@n31, InValue=1)
   SendDebugValue(execute=@n30.else, ParameterName=Rough_OverrideOrAdjust, Value=0)
   > SetMinValue(self=@n31, InValue=0)
   > SetMaxValue(self=@n31, InValue=1)
   n32 = "On Check State Changed (MetalMode)"!() -> (OutputDelegate, then, bIsChecked:bool) {is=title K2Node_ComponentBoundEvent}
   > n33 = Branch(Condition=@n32.bIsChecked)
   > SendDebugValue(ParameterName=Metal_OverrideOrAdjust, Value=1)
   n34 = Get MetalSlider() -> (MetalSlider:Editor Utility Slider Object Reference)
   > SetMinValue(self=@n34, InValue=-1)
   > SetMaxValue(self=@n34, InValue=1)
   SendDebugValue(execute=@n33.else, ParameterName=Metal_OverrideOrAdjust, Value=0)
   > SetMinValue(self=@n34, InValue=0)
   > SetMaxValue(self=@n34, InValue=1)
   n35 = "On Value Changed (RoughSlider)"!() -> (OutputDelegate, then, Value:float) {is=title K2Node_ComponentBoundEvent}
   > SendDebugValue(ParameterName=Roughness, Value=@n35.Value)
   > Set Text(
    Text="To Text (Float)"(Value=@n35.Value, MinimumIntegralDigits=1, MaximumIntegralDigits=324, MinimumFractionalDigits=2, MaximumFractionalDigits=2)
    self=Get RoughValue() -> (RoughValue:Text Object Reference)
   )
   n36 = "On Value Changed (MetalSlider)"!() -> (OutputDelegate, then, Value:float) {is=title K2Node_ComponentBoundEvent}
   > SendDebugValue(ParameterName=Metallic, Value=@n36.Value)
   > Set Text(
    Text="To Text (Float)"(Value=@n36.Value, MinimumIntegralDigits=1, MaximumIntegralDigits=324, MinimumFractionalDigits=2, MaximumFractionalDigits=2)
    self=Get MetalValue() -> (MetalValue:Text Object Reference)
   )
   n37 = "On Value Changed (SpecSlider)"!() -> (OutputDelegate, then, Value:float) {is=title K2Node_ComponentBoundEvent}
   > SendDebugValue(ParameterName=Specular, Value=@n37.Value)
   > Set Text(
    Text="To Text (Float)"(Value=@n37.Value, MinimumIntegralDigits=1, MaximumIntegralDigits=324, MinimumFractionalDigits=2, MaximumFractionalDigits=2)
    self=Get SpecValue() -> (SpecValue:Text Object Reference)
   )
   n38 = SelectionChanged() {is=title K2Node_CustomEvent}
   Unbind(execute=[@n38.then, @n52.then])
   > Clear(TargetArray:Array of Actor Object References=Get CurrentSelection() -> (CurrentSelection:Array of Actor Object References))
   > Set IsDebugging(IsDebugging:bool=false) -> (Output_Get:bool)
   > SetIsTemporarilyHiddenInEditor!(execute, self:Actor Object Reference=Get DebugTarget() -> (DebugTarget:Actor Object Reference), bIsHidden:bool=false) -> (then)
   n39 = Get TA_Generic01_Button() -> (TA_Generic01_Button:TA Generic 01 Button Object Reference)
   > setLabel!(execute, self:TA Base Editor Button Widget Object Reference=@n39, label:String=Duplicate and Review Selected) -> (then)
   > Set buttonColor_Default!(execute, buttonColor_Default:Linear Color Structure=(R=0.1,G=0.45,B=1,A=1), self:TA Base Editor Button Widget Object Reference=@n39) -> (then, Output_Get:Linear Color Structure)
   > Destroy Actor(self=Get TransientDuplicate() -> (TransientDuplicate:Actor Object Reference))
   > Clear(TargetArray:Array of Material Instance Dynamic Object References=Get TrackedMaterialsForDebug())
   > ResetSliders()
   n40 = SelectionChange!() -> (OutputDelegate, then, SelectionSet:Typed Element Selection Set Object Reference) {is=title K2Node_CustomEvent}
   > n41 = GetSelectedObjects!(execute, self:Typed Element Selection Set Object Reference=@n40.SelectionSet, InRequiredClass:Object Class Reference=&Actor) -> (then, ReturnValue:Array of Object References)
   > n42 = For Each Loop(Array:Array of Object References=@n41) -> (Array Element:Object Reference)
   > PrintString(InString=GetDisplayName(Object=@n42))
   Bind() {is=title K2Node_CustomEvent}
   > n43 = GetSelectionSet(self=Get LevelEditorSubsystem())
   > Bind Event to On Selection Change(self:Typed Element Selection Set Object Reference=@n43, Delegate=@n40) {is=title K2Node_AddDelegate}
   Unbind() {is=title K2Node_CustomEvent}
   > Unbind Event from On Selection Change(self:Typed Element Selection Set Object Reference=@n43, Delegate=@n40) {is=title K2Node_RemoveDelegate}
   Reroute Node()
   > n44 = Gather Subobject Data For Instance(self=Get SubobjectDataSubsystem())
   > n45 = For Each Loop(Array:Array of Subobject Data Handle Structures=@n44) -> (Array Element:Subobject Data Handle Structure)
   > n46 = Cast To StaticMeshComponent(Object=GetAssociatedObject(Data=GetData(DataHandle=@n45)))
   > n47 = GetAllParentMIsToBaseM_Recursive(Material interface in=GetMaterial(self=@n46, ElementIndex=0))
   > SetMaterial(self=@n46, ElementIndex=0, Material=@n47)
   GetDisplayName(Object=@n10)
   PrintString(InString=Hello)
   Event PreConstruct()
   Event Construct()
   Event Tick()
   Get LevelEditorSubsystem()
   Get LevelEditorUISubsystem() -> (ReturnValue:Level Editor UISubsystem Object Reference) {is=title K2Node_GetEditorSubsystem}
   Get EditorActorSubsystem()
   Get SpecMode() -> (SpecMode:Editor Utility Check Box Object Reference)
   Get SpecSlider()
   Get SpecValue()
   SetIsTemporarilyHiddenInEditor(execute=@n4.then, self=@n4, bIsHidden=true)
   > setLabel(self=@n39, label=Append(A="Debugging ", B=GetDisplayName(Object=@n4)))
   > n48 = Set buttonColor_Default(buttonColor_Default=(R=0.3438,G=1,B=0.3229,A=1), self=@n39)
   n49 = CheckIsValidForDebug!(execute=@n51.then, self, Context:Actor Object Reference=@n3) -> (then, IsValid?:bool) {is=title K2Node_CallFunction}
   > Branch(Condition=@n49)
   > n50 = Set IsDebugging(IsDebugging=true)
   n51 = Bind(execute=@n53.then) {is=title K2Node_CallFunction}
   n52 = Branch(execute=@n2.then, Condition=Get IsDebugging() -> (IsDebugging:bool))
   n53 = Set CurrentSelection(execute=@n52.else, CurrentSelection:Array of Actor Object References=@n2) -> (Output_Get:Array of Actor Object References)
  ]}
 GetAllParentMIsToBaseM_Recursive {nodes [
   n1 = GetAllParentMIsToBaseM_Recursive!() -> (then, Material interface in:Material Interface Object Reference) {is=title K2Node_FunctionEntry}
   > n2 = Cast To Material!(execute, Object=@n1) -> (then, CastFailed, AsMaterial:Material Object Reference) {is=title K2Node_DynamicCast}
   > n3 = CreateUniqueNewAssetPathName!(
    execute
    AssetFolderPath:String=/Temp/
    BaseAssetName:String=GetDisplayName(Object=@n2)
    Options:Geometry Script Unique Asset Name Options Structure
   ) -> (Failure:Exec, Success:Exec, UniqueAssetPathAndName:String, UniqueAssetName:String)
   n4 = SpawnTransientObject!(execute=@n3.Success, ObjectClass:Object Class Reference=/Script/Engine.Material, Parent:Object Reference=@n2) -> (then, SpawnedObject:Material Object Reference)
   > n5 = DuplicateObject!(execute, SourceObject:Object Reference=@n2, Outer:Object Reference=@n4, Name:Name=String To Name(InString=@n3.UniqueAssetName)) -> (then, ReturnValue:Object Reference)
   > n6 = Cast To Material(Object=@n5)
   > n7 = CreateMaterialExpression!(
    execute
    Material:Material Object Reference=@n6
    ExpressionClass:Material Expression Class Reference=/Script/Engine.MaterialExpressionMaterialFunctionCall
    NodePosX:int=0
    NodePosY:int=0
   ) -> (then, ReturnValue:Material Expression Object Reference)
   > n8 = Cast To MaterialExpressionMaterialFunctionCall!(execute, Object=@n7) -> (then, CastFailed, AsMaterial Expression Material Function Call:Material Expression Material Function Call Object Reference) {is=title K2Node_DynamicCast}
   > SetMaterialFunction!(
    execute
    self:Material Expression Material Function Call Object Reference=@n8
    NewMaterialFunction:Material Function Interface Object Reference=/Game/NeverCook/ArtTools/Tech_Art/ArtReview_Visualizer/PostProcess_Mats/MF_ArtReview_Debugger
   ) -> (then, ReturnValue:bool)
   n9 = GetMaterialPropertyInputNode!(Material:Material Object Reference=@n6, Property:EMaterialProperty Enum=MP_MaterialAttributes) -> (ReturnValue:Material Expression Object Reference)
   > n10 = Is Valid(InputObject=@n9)
   ConnectMaterialExpressions!(
    execute=@n10.Is Not Valid
    FromExpression:Material Expression Object Reference=GetMaterialPropertyInputNode(Material=@n6, Property=MP_Roughness)
    FromOutputName:String=GetMaterialPropertyInputNodeOutputName!(Material:Material Object Reference=@n6, Property:EMaterialProperty Enum=MP_Roughness) -> (ReturnValue:String)
    ToExpression:Material Expression Object Reference=@n8
    ToInputName:String=Roughness
   ) -> (then, ReturnValue:bool)
   > ConnectMaterialExpressions(
    FromExpression=GetMaterialPropertyInputNode(Material=@n6, Property=MP_Specular)
    FromOutputName=GetMaterialPropertyInputNodeOutputName(Material=@n6, Property=MP_Specular)
    ToExpression=@n8
    ToInputName=Specular
   )
   > ConnectMaterialExpressions(
    FromExpression=GetMaterialPropertyInputNode(Material=@n6, Property=MP_Metallic)
    FromOutputName=GetMaterialPropertyInputNodeOutputName(Material=@n6, Property=MP_Metallic)
    ToExpression=@n8
    ToInputName=Metallic
   )
   > ConnectMaterialProperty!(
    execute
    FromExpression:Material Expression Object Reference=@n8
    FromOutputName:String=Roughness
    Property:EMaterialProperty Enum=MP_Roughness
   ) -> (then, ReturnValue:bool)
   > ConnectMaterialProperty(FromExpression=@n8, FromOutputName=Specular, Property=MP_Specular)
   > ConnectMaterialProperty(FromExpression=@n8, FromOutputName=Metallic, Property=MP_Metallic)
   > n11 = RecompileMaterial!(execute, Material:Material Object Reference=@n6) -> (then, ReturnValue:Array of Strings)
   Return Node!(execute=[@n11.then, @n12.then], InputTempVersion:Material Interface Object Reference=@n6)
   ConnectMaterialExpressions(
    execute=@n10.Is Valid
    FromExpression=@n9
    FromOutputName=GetMaterialPropertyInputNodeOutputName(Material=@n6, Property=MP_MaterialAttributes)
    ToExpression=@n8
    ToInputName=In
   )
   > ConnectMaterialProperty(FromExpression=@n8, FromOutputName=MaterialAttributes, Property=MP_MaterialAttributes)
   > n12 = RecompileMaterial(Material=@n6)
   n13 = Cast To MaterialInstanceDynamic!(execute=@n2.CastFailed, Object=@n1) -> (then, CastFailed, AsMaterial Instance Dynamic:Material Instance Dynamic Object Reference) {is=title K2Node_DynamicCast}
   > n14 = GetAllParentMIsToBaseM_Recursive(Material interface in=Get Parent!(self:Material Instance Object Reference=@n13) -> (Parent:Material Interface Object Reference)) {is=title K2Node_CallFunction}
   > n15 = CreateDynamicMaterialInstance(Parent=@n14, OptionalName=None, CreationFlags=None) {is=title K2Node_CallMaterialParameterCollectionFunction}
   > SetEditorProperty(
    Object=@n15
    PropertyName=ScalarParameterValues
    PropertyValue:Array of Scalar Parameter Value Structures=Get ScalarParameterValues!(self:Material Instance Object Reference=@n13) -> (ScalarParameterValues:Array of Scalar Parameter Value Structures)
   )
   > SetEditorProperty(
    Object=@n15
    PropertyName=VectorParameterValues
    PropertyValue:Array of Vector Parameter Value Structures=Get VectorParameterValues!(self:Material Instance Object Reference=@n13) -> (VectorParameterValues:Array of Vector Parameter Value Structures)
   )
   > SetEditorProperty(
    Object=@n15
    PropertyName=TextureParameterValues
    PropertyValue:Array of Texture Parameter Value Structures=Get TextureParameterValues!(self:Material Instance Object Reference=@n13) -> (TextureParameterValues:Array of Texture Parameter Value Structures)
   )
   > Return Node(InputTempVersion=@n15)
   n16 = Cast To MaterialInstance!(execute=@n13.CastFailed, Object=@n1) -> (then, CastFailed, AsMaterial Instance:Material Instance Object Reference) {is=title K2Node_DynamicCast}
   > n17 = CreateUniqueNewAssetPathName(AssetFolderPath=/Temp/, BaseAssetName=GetDisplayName(Object=@n16))
   n18 = SpawnTransientObject(execute=@n17.Success, ObjectClass=/Script/Engine.MaterialInstanceConstant, Parent=@n16) -> (SpawnedObject:Material Instance Constant Object Reference)
   > n19 = DuplicateObject(SourceObject=@n16, Outer=@n18, Name=String To Name(InString=@n17.UniqueAssetName))
   > n20 = GetAllParentMIsToBaseM_Recursive(Material interface in=Get Parent(self=@n16))
   > n21 = Cast To MaterialInstanceConstant(Object=@n19)
   > SetMaterialInstanceParent!(execute, Instance:Material Instance Constant Object Reference=@n21, NewParent:Material Interface Object Reference=@n20) -> (then)
   > Return Node(InputTempVersion=@n21)
  ]}
 ResetSliders {nodes [
   ResetSliders() {is=title K2Node_FunctionEntry}
   > n1 = Sequence()
   n2 = Get SpecSlider()
   > SetMinValue(self=@n2, InValue=0)
   > SetMaxValue(self=@n2, InValue=1)
   > Set Value!(execute, Value:float=0.5, self:Slider Object Reference=@n2) -> (then, Output_Get:float)
   n3 = Get MetalSlider()
   > SetMinValue(self=@n3, InValue=0)
   > SetMaxValue(self=@n3, InValue=1)
   > Set Value(Value=0.5, self=@n3)
   n4 = Get RoughSlider()
   > SetMinValue(self=@n4, InValue=0)
   > SetMaxValue(self=@n4, InValue=1)
   > Set Value(Value=0.5, self=@n4)
   Set CheckedState!(
    execute=@n1.then_1
    CheckedState:ECheckBoxState Enum=Unchecked
    self:Check Box Object Reference=Get Specular() -> (Specular:Editor Utility Check Box Object Reference)
   ) -> (then, Output_Get:ECheckBoxState Enum=Unchecked)
   > Set CheckedState(CheckedState=Unchecked, self=Get Roughness() -> (Roughness:Editor Utility Check Box Object Reference)) -> (Output_Get=Unchecked)
   > Set CheckedState(CheckedState=Unchecked, self=Get Metallic() -> (Metallic:Editor Utility Check Box Object Reference)) -> (Output_Get=Unchecked)
   > Set CheckedState(CheckedState=Unchecked, self=Get RoughMode() -> (RoughMode:Editor Utility Check Box Object Reference)) -> (Output_Get=Unchecked)
   > Set CheckedState(CheckedState=Unchecked, self=Get SpecMode()) -> (Output_Get=Unchecked)
   > Set CheckedState(CheckedState=Unchecked, self=Get MetalMode() -> (MetalMode:Editor Utility Check Box Object Reference)) -> (Output_Get=Unchecked)
  ]}
 SendDebugValue {nodes [
   n1 = SendDebugValue!() -> (then, ParameterName:Name, Value:float=~) {is=title K2Node_FunctionEntry}
   > n2 = For Each Loop(Array:Array of Material Instance Dynamic Object References=Get TrackedMaterialsForDebug()) -> (Array Element:Material Instance Dynamic Object Reference)
   > SetScalarParameterValue(self=@n2, ParameterName=@n1, Value=@n1.Value)
  ]}
]
variables [
 CurrentSelection {type="array<Actor>"}
 DebugTarget {type=Actor, default=null}
 IsDebugging {type=boolean}
 TrackedMaterialsForDebug {type="array<MaterialInstanceDynamic>", default []}
 TransientDuplicate {type=Actor, default=null}
]
widgets {
 info {rootWidgetClass=&Overlay, widgetCount=28}
 widgets [
  Overlay_17 {widgetClassPath=&Overlay, widget=Overlay_17}
  Border_38 {
   widgetClassPath=&Border
   parent=Overlay_17
   widget=Border_38
   properties {verticalAlignment=VAlign_Top, brushColor=<rgba 0.1562 0.1562 0.1562 1>}
   slot {class=OverlaySlot, properties {horizontalAlignment=HAlign_Fill}}
  }
  HorizontalBox_35 {
   widgetClassPath=&HorizontalBox
   parent=Border_38
   widget=HorizontalBox_35
   slot {class=BorderSlot, properties {verticalAlignment=VAlign_Top}}
  }
  VerticalBox_1 {
   widgetClassPath=&VerticalBox
   parent=HorizontalBox_35
   widget=VerticalBox_1
   slot {class=HorizontalBoxSlot, properties {size {sizeRule=Fill, value=2}}}
  }
  TA_Generic01_Button {
   widgetClassPath=&TA_Generic01_Button
   parent=VerticalBox_1
   widget=TA_Generic01_Button
   properties {label=Duplicate and Review Selected}
   slot {class=VerticalBoxSlot, properties {size {sizeRule=Fill, value=1}}}
  }
  AdvancedSliders {
   widgetClassPath=&VerticalBox
   bIsVariable=true
   parent=VerticalBox_1
   widget=AdvancedSliders
   properties {bOverride_Cursor=true}
   slot {class=VerticalBoxSlot, properties {size {sizeRule=Fill, value=2}}}
  }
  SpecularBox {
   widgetClassPath=&HorizontalBox
   parent=AdvancedSliders
   widget=SpecularBox
   slot {class=VerticalBoxSlot, properties {size {sizeRule=Fill, value=1}}}
  }
  Specular {
   widgetClassPath=&EditorUtilityCheckBox
   bIsVariable=true
   parent=SpecularBox
   widget=Specular
   slot {class=HorizontalBoxSlot}
  }
  TextBlock_223 {
   widgetClassPath=&TextBlock
   parent=Specular
   widget=TextBlock_223
   properties {
    text=Spec Override
    font {
     bForceMonospaced=false
     bMaterialIsStencil=false
     fontMaterial=null
     fontObject=&Roboto
     letterSpacing=0
     monospacedWidth=1
     size=11.25
     skewAmount=0
     typefaceFontName=Bold
     outlineSettings {
      bApplyOutlineToDropShadows=false
      bMiteredCorners=false
      bSeparateFillAlpha=false
      outlineMaterial=null
      outlineSize=0
      outlineColor=<rgba 0 0 0 1>
     }
    }
   }
   slot {class=PanelSlot}
  }
  SpecMode {
   widgetClassPath=&EditorUtilityCheckBox
   bIsVariable=true
   parent=SpecularBox
   widget=SpecMode
   slot {class=HorizontalBoxSlot}
  }
  TextBlock_2 {
   widgetClassPath=&TextBlock
   parent=SpecMode
   widget=TextBlock_2
   properties {
    text=Add Instead of Override
    font {
     bForceMonospaced=false
     bMaterialIsStencil=false
     fontMaterial=null
     fontObject=&Roboto
     letterSpacing=0
     monospacedWidth=1
     size=11.25
     skewAmount=0
     typefaceFontName=Bold
     outlineSettings {
      bApplyOutlineToDropShadows=false
      bMiteredCorners=false
      bSeparateFillAlpha=false
      outlineMaterial=null
      outlineSize=0
      outlineColor=<rgba 0 0 0 1>
     }
    }
   }
   slot {class=PanelSlot}
  }
  SpecValue {
   widgetClassPath=&TextBlock
   bIsVariable=true
   parent=SpecularBox
   widget=SpecValue
   properties {
    text=0.50
    font {
     bForceMonospaced=false
     bMaterialIsStencil=false
     fontMaterial=null
     fontObject=&Roboto
     letterSpacing=0
     monospacedWidth=1
     size=14.25
     skewAmount=0
     typefaceFontName=Bold
     outlineSettings {
      bApplyOutlineToDropShadows=false
      bMiteredCorners=false
      bSeparateFillAlpha=false
      outlineMaterial=null
      outlineSize=0
      outlineColor=<rgba 0 0 0 1>
     }
    }
   }
   slot {class=HorizontalBoxSlot, properties {horizontalAlignment=HAlign_Left, padding=<ltrb 2 0 45 0>, size {sizeRule=Fill, value=0}}}
  }
  SpecSlider {
   widgetClassPath=/Script/Blutility.EditorUtilitySlider
   bIsVariable=true
   parent=SpecularBox
   widget=SpecSlider
   properties {value=0.5}
   slot {class=HorizontalBoxSlot, properties {size {sizeRule=Fill, value=1}}}
  }
  MetallicBox {
   widgetClassPath=&HorizontalBox
   parent=AdvancedSliders
   widget=MetallicBox
   slot {class=VerticalBoxSlot, properties {size {sizeRule=Fill, value=1}}}
  }
  HorizontalBox_0 {
   widgetClassPath=&HorizontalBox
   parent=MetallicBox
   widget=HorizontalBox_0
   slot {class=HorizontalBoxSlot, properties {size {sizeRule=Fill, value=1}}}
  }
  Metallic {
   widgetClassPath=&EditorUtilityCheckBox
   bIsVariable=true
   parent=HorizontalBox_0
   widget=Metallic
   slot {class=HorizontalBoxSlot}
  }
  TextBlock {
   widgetClassPath=&TextBlock
   parent=Metallic
   widget=TextBlock
   properties {
    text=Metal Override
    font {
     bForceMonospaced=false
     bMaterialIsStencil=false
     fontMaterial=null
     fontObject=&Roboto
     letterSpacing=0
     monospacedWidth=1
     size=11.25
     skewAmount=0
     typefaceFontName=Bold
     outlineSettings {
      bApplyOutlineToDropShadows=false
      bMiteredCorners=false
      bSeparateFillAlpha=false
      outlineMaterial=null
      outlineSize=0
      outlineColor=<rgba 0 0 0 1>
     }
    }
   }
   slot {class=PanelSlot}
  }
  MetalMode {
   widgetClassPath=&EditorUtilityCheckBox
   bIsVariable=true
   parent=HorizontalBox_0
   widget=MetalMode
   slot {class=HorizontalBoxSlot}
  }
  TextBlock_3 {
   widgetClassPath=&TextBlock
   parent=MetalMode
   widget=TextBlock_3
   properties {
    text=Add Instead of Override
    font {
     bForceMonospaced=false
     bMaterialIsStencil=false
     fontMaterial=null
     fontObject=&Roboto
     letterSpacing=0
     monospacedWidth=1
     size=11.25
     skewAmount=0
     typefaceFontName=Bold
     outlineSettings {
      bApplyOutlineToDropShadows=false
      bMiteredCorners=false
      bSeparateFillAlpha=false
      outlineMaterial=null
      outlineSize=0
      outlineColor=<rgba 0 0 0 1>
     }
    }
   }
   slot {class=PanelSlot}
  }
  MetalValue {
   widgetClassPath=&TextBlock
   bIsVariable=true
   parent=HorizontalBox_0
   widget=MetalValue
   properties {
    text=0.50
    font {
     bForceMonospaced=false
     bMaterialIsStencil=false
     fontMaterial=null
     fontObject=&Roboto
     letterSpacing=0
     monospacedWidth=1
     size=14.25
     skewAmount=0
     typefaceFontName=Bold
     outlineSettings {
      bApplyOutlineToDropShadows=false
      bMiteredCorners=false
      bSeparateFillAlpha=false
      outlineMaterial=null
      outlineSize=0
      outlineColor=<rgba 0 0 0 1>
     }
    }
   }
   slot {class=HorizontalBoxSlot}
  }
  MetalSlider {
   widgetClassPath=&EditorUtilitySlider
   bIsVariable=true
   parent=HorizontalBox_0
   widget=MetalSlider
   properties {value=0.5}
   slot {class=HorizontalBoxSlot, properties {size {sizeRule=Fill, value=1}}}
  }
  RoughnessBox {
   widgetClassPath=&HorizontalBox
   parent=AdvancedSliders
   widget=RoughnessBox
   slot {class=VerticalBoxSlot, properties {size {sizeRule=Fill, value=1}}}
  }
  Roughness {
   widgetClassPath=&EditorUtilityCheckBox
   bIsVariable=true
   parent=RoughnessBox
   widget=Roughness
   slot {class=HorizontalBoxSlot}
  }
  TextBlock_1 {
   widgetClassPath=&TextBlock
   parent=Roughness
   widget=TextBlock_1
   properties {
    text=Roughness Override
    font {
     bForceMonospaced=false
     bMaterialIsStencil=false
     fontMaterial=null
     fontObject=&Roboto
     letterSpacing=0
     monospacedWidth=1
     size=11.25
     skewAmount=0
     typefaceFontName=Bold
     outlineSettings {
      bApplyOutlineToDropShadows=false
      bMiteredCorners=false
      bSeparateFillAlpha=false
      outlineMaterial=null
      outlineSize=0
      outlineColor=<rgba 0 0 0 1>
     }
    }
   }
   slot {class=PanelSlot}
  }
  RoughMode {
   widgetClassPath=&EditorUtilityCheckBox
   bIsVariable=true
   parent=RoughnessBox
   widget=RoughMode
   slot {class=HorizontalBoxSlot}
  }
  TextBlock_4 {
   widgetClassPath=&TextBlock
   parent=RoughMode
   widget=TextBlock_4
   properties {
    text=Add Instead of Override
    font {
     bForceMonospaced=false
     bMaterialIsStencil=false
     fontMaterial=null
     fontObject=&Roboto
     letterSpacing=0
     monospacedWidth=1
     size=11.25
     skewAmount=0
     typefaceFontName=Bold
     outlineSettings {
      bApplyOutlineToDropShadows=false
      bMiteredCorners=false
      bSeparateFillAlpha=false
      outlineMaterial=null
      outlineSize=0
      outlineColor=<rgba 0 0 0 1>
     }
    }
   }
   slot {class=PanelSlot}
  }
  RoughValue {
   widgetClassPath=&TextBlock
   bIsVariable=true
   parent=RoughnessBox
   widget=RoughValue
   properties {
    text=0.50
    font {
     bForceMonospaced=false
     bMaterialIsStencil=false
     fontMaterial=null
     fontObject=&Roboto
     letterSpacing=0
     monospacedWidth=1
     size=14.25
     skewAmount=0
     typefaceFontName=Bold
     outlineSettings {
      bApplyOutlineToDropShadows=false
      bMiteredCorners=false
      bSeparateFillAlpha=false
      outlineMaterial=null
      outlineSize=0
      outlineColor=<rgba 0 0 0 1>
     }
    }
   }
   slot {class=HorizontalBoxSlot, properties {padding=<ltrb 2 0 45 0>, size {sizeRule=Fill, value=0}}}
  }
  RoughSlider {
   widgetClassPath=&EditorUtilitySlider
   bIsVariable=true
   parent=RoughnessBox
   widget=RoughSlider
   properties {value=0.5}
   slot {class=HorizontalBoxSlot, properties {size {sizeRule=Fill, value=1}}}
  }
 ]
}
