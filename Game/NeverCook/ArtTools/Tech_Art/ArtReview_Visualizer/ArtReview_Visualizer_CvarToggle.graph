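# Editor utility widget in the art-review visualizer tools that exposes a few console variables.
# A vertical box stacks reusable cvar spin boxes for motion blur amount and near clip plane above a checkbox for selection outline, each configured through instance properties.
# The event graph holds only unconnected fragments: empty events, BlurAmountCvar setters, and a generic cvar readout referencing text widgets absent from the tree.
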EditorUtilityWidgetBlueprint /Game/NeverCook/ArtTools/Tech_Art/ArtReview_Visualizer/ArtReview_Visualizer_CvarToggle
graphs [EventGraph {nodes [
   n1 = Get Cvar() -> (Cvar:String)
   Set Text(Text="To Text (String)"(InString=@n1), self=Get TextBlock_71() -> (TextBlock_71:Text Object Reference))
   n2 = GetNumericCVar!(CVar:String=@n1) -> (OutValue:float, ReturnValue:bool)
   > n3 = Branch(Condition=@n2.ReturnValue)
   > Set Text(
    Text="To Text (Float)"(Value=@n2, MinimumIntegralDigits=1, MaximumIntegralDigits=324, MinimumFractionalDigits=0, MaximumFractionalDigits=3)
    self:Editable Text Object Reference=Get EditorUtilityEditableText_83() -> (EditorUtilityEditableText_83:Editor Utility Editable Text Object Reference)
   )
   PrintString(execute=@n3.else, InString="Error, bad Cvar")
   Event PreConstruct()
   Event Construct()
   Event Tick()
   Get DataDrivenCVarEngineSubsystem() -> (ReturnValue:Data Driven CVars Object Reference) {is=title K2Node_GetEngineSubsystem}
   Get DataDrivenCVarEngineSubsystem()
   n4 = Get BlurAmountCvar() -> (BlurAmountCvar:EUW Cvar Spin Box Object Reference)
   Set min!(execute, min:double=-1, self:EUW Cvar Spin Box Object Reference=@n4) -> (then, Output_Get:double)
   Set settingUp!(execute, settingUp:bool=true, self:EUW Cvar Spin Box Object Reference=@n4) -> (then, Output_Get:bool)
  ]}]
variables [Cvar {type=string, default=r.ScreenPercentage}]
widgets {
 info {rootWidgetClass=&VerticalBox, widgetCount=4}
 widgets [
  VerticalBox_0 {widgetClassPath=&VerticalBox, widget=VerticalBox_0}
  BlurAmountCvar {
   widgetClassPath=/Game/ArtTools/RenderSettings/EUW_cvar_spinBox
   bIsVariable=true
   parent=VerticalBox_0
   widget=BlurAmountCvar
   properties {cvar=r.MotionBlur.Amount, labelText=r.MotionBlur.Amount, min=-1, useLoadingValueAsDefault=true}
   slot {class=VerticalBoxSlot}
  }
  NearClipPlane {
   widgetClassPath=&EUW_cvar_spinBox
   bIsVariable=true
   parent=VerticalBox_0
   widget=NearClipPlane
   properties {
    cvar=r.SetNearClipPlane
    defaultValue=20
    labelText=r.SetNearClipPlane
    labelToolTip="Sets the near clip plane, default is 20"
    max=20
    step=0.5
   }
   slot {class=VerticalBoxSlot}
  }
  SelectionboxCvar {
   widgetClassPath=/Game/ArtTools/RenderSettings/EUW_cvar_checkBox
   bIsVariable=true
   parent=VerticalBox_0
   widget=SelectionboxCvar
   properties {cvar=ShowFlag.SelectionOutline, labelText=ShowFlag.SelectionOutline}
   slot {class=VerticalBoxSlot, properties {horizontalAlignment=HAlign_Left, verticalAlignment=VAlign_Center}}
  }
 ]
}
