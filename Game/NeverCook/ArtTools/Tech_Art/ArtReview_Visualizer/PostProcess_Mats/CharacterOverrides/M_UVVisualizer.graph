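# Unlit debug material for art-review character overrides that visualizes UV layout by color-coding which tile the second UV channel (index 1) falls in.
# Nested If nodes band U into red, green, blue and yellow at whole-number boundaries, with a muted shade where V is above zero.
# Separate debug-digit function branches print placeholder constants over repeating UV0 cells.
# The graph shows no connection from any branch to a material output.
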Material /Game/NeverCook/ArtTools/Tech_Art/ArtReview_Visualizer/PostProcess_Mats/CharacterOverrides/M_UVVisualizer
graph {
 properties {
  bAutomaticallySetUsageInEditor=false
  previewMesh=/Hermes/Sets/UtopiaCity/Meshes/SM_UtopiaCity_Double_Doorway_B_A
  shadingModel=MSM_Unlit
  usedShadingModels=Unlit
 }
 nodes [
  Constant3Vector() {constant=<rgba 0 0 1 1>}
  Constant3Vector() {constant=<rgba 1 1 0 1>}
  Constant3Vector() {constant=<rgba 0 1 0 1>}
  Constant3Vector() {constant=<rgba 1 0 0 1>}
  Constant()
  Constant() {r=1}
  Constant() {r=1}
  Constant()
  Constant() {r=1}
  Constant()
  Constant() {r=1}
  Constant()
  n1 = ComponentMask(None=TextureCoordinate() {coordinateIndex=1}) {r=true}
  n2 = Constant3Vector() {constant=<rgba 1 1 0 1>}
  n3 = If(
   A=ComponentMask(None=TextureCoordinate() {coordinateIndex=1}) {g=true}
   "A > B"=Constant3Vector() {constant=<rgba 0.5 0.5 0.2 1>}
   "A == B"=@n2
   "A < B"=@n2
  )
  n4 = Constant3Vector() {constant=<rgba 0 0 1 1>}
  n5 = If(
   A=ComponentMask(None=TextureCoordinate() {coordinateIndex=1}) {g=true}
   "A > B"=Constant3Vector() {constant=<rgba 0.25 0.25 0.5 1>}
   "A == B"=@n4
   "A < B"=@n4
  )
  n6 = Constant3Vector() {constant=<rgba 0 1 0 1>}
  n7 = If(
   A=ComponentMask(None=TextureCoordinate() {coordinateIndex=1}) {g=true}
   "A > B"=Constant3Vector() {constant=<rgba 0.25 0.4 0.25 1>}
   "A == B"=@n6
   "A < B"=@n6
  )
  n8 = Constant3Vector() {constant=<rgba 1 0 0 1>}
  n9 = If(
   A=ComponentMask(None=TextureCoordinate() {coordinateIndex=1}) {g=true}
   "A > B"=Constant3Vector() {constant=<rgba 0.5 0.25 0.25 1>}
   "A == B"=@n8
   "A < B"=@n8
  )
  If(
   A=@n1
   "A > B"=If(A=@n1, "A > B"=If(A=@n1, "A > B"=@n3, "A == B"=@n5, "A < B"=@n5) {constB=3, desc=U 2-3}, "A == B"=@n7, "A < B"=@n7) {constB=2, desc=U 1-2}
   "A == B"=@n9
   "A < B"=@n9
  ) {constB=1, desc=U 0-1}
  n10 = Constant() {r=2}
  n11 = TextureCoordinate()
  MaterialFunctionCall(Number=Constant() {r=1}, MaximumNumberOfDigits=@n10, UVs=Modulo!(A=Multiply(A=@n11) {constB=20}, B=@n10)) {materialFunction=&DebugScalarValues}
  n12 = Constant() {r=2}
  n13 = TextureCoordinate()
  MaterialFunctionCall!(
   Vector2=Constant2Vector()
   MaximumNumberOfDigits=@n12
   UVs=Modulo(A=Multiply(A=@n13) {constB=30}, B=@n12)
   DebugTextLocation RG_UpperRight BA_LowerLeft=Constant4Vector() {constant=<rgba 0 0 1 1>}.RGBA
   Component Spacing=Constant4Vector() {constant=<rgba 0 0 1 0.8>}.RGBA
  ) {materialFunction=/Engine/Functions/Engine_MaterialFunctions02/Utility/DebugFloat2Values}
 ]
}
