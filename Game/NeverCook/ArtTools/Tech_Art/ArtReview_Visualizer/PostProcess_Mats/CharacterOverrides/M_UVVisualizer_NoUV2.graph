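# M_UVVisualizer_NoUV2 is an unlit override material in the art-review visualizer's character overrides, named for the no-second-UV case of the UV visualizer.
# Its graph holds a single constant dark grey color, with no UV sampling, textures or parameters.
# Automatic usage flagging in the editor is turned off.
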Material /Game/NeverCook/ArtTools/Tech_Art/ArtReview_Visualizer/PostProcess_Mats/CharacterOverrides/M_UVVisualizer_NoUV2
graph {
 properties {bAutomaticallySetUsageInEditor=false, shadingModel=MSM_Unlit, usedShadingModels=Unlit}
 nodes [Constant3Vector() {constant=<rgba 0.0281 0.0281 0.0281 0>}]
}
