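# Unlit surface material in the ArtReview Visualizer tools, named for viewing vertex color alpha during art review.
# Its graph holds a single Vertex Color node, with no connection to a material output shown.
# Usage flags cover skeletal meshes, morph targets and mesh deformers.
# Despite sitting in a PostProcess_Mats folder, it keeps the default material domain.
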Material /Game/NeverCook/ArtTools/Tech_Art/ArtReview_Visualizer/PostProcess_Mats/MAT_artReview_VertexColorAlpha
graph {
 properties {
  bUsedWithMeshDeformer=true
  bUsedWithMorphTargets=true
  bUsedWithSkeletalMesh=true
  pixelDepthOffsetMode=PDOM_Legacy
  shadingModel=MSM_Unlit
  usedShadingModels=Unlit
 }
 nodes [VertexColor()]
}
